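# Material for the SuperSwirl skydiving contrail, bending SplineThicken strands into curling swirls with world position offset.
# Custom HLSL takes particle and origin positions from dynamic parameters to derive swirl center, axis and radius, then adds roll, stretch and upward bend.
# MX_RandomFloat varies swirl shape and hue per particle.
# Strands taper at both ends and near the origin, with camera depth fade and time-of-day color.
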Material /BRCosmetics/Athena/Effects/Particles/Skydiving/NS_Contrail_SuperSwirl/Materials/M_Contrail_SuperSwirl
graph {
 expressions {
  MaterialExpressionAdd_0 {
   class=MaterialExpressionAdd
   inputs [
    A {from=MaterialExpressionNamedRerouteUsage_11}
    B {from=MaterialExpressionNamedRerouteUsage_9}
   ]
  }
  MaterialExpressionAdd_1 {
   class=MaterialExpressionAdd
   inputs [
    A {from=MaterialExpressionNamedRerouteUsage_14}
    B {from=MaterialExpressionConstant_1}
   ]
  }
  MaterialExpressionAdd_10 {
   class=MaterialExpressionAdd
   inputs [
    A {from=MaterialExpressionMultiply_39}
    B {from=""}
   ]
   properties {constB=0.5}
  }
  MaterialExpressionAdd_11 {
   class=MaterialExpressionAdd
   inputs [
    A {from=MaterialExpressionAdd_13}
    B {from=""}
   ]
   properties {constB=10, desc=overall fade in}
  }
  MaterialExpressionAdd_12 {
   class=MaterialExpressionAdd
   inputs [
    A {from=MaterialExpressionReroute_6}
    B {from=MaterialExpressionMultiply_38}
   ]
  }
  MaterialExpressionAdd_13 {
   class=MaterialExpressionAdd
   inputs [
    A {from=MaterialExpressionMultiply_1}
    B {from=""}
   ]
   properties {constB=5, desc=gap}
  }
  MaterialExpressionAdd_14 {
   class=MaterialExpressionAdd
   inputs [
    A {from=MaterialExpressionNamedRerouteUsage_38}
    B {from=MaterialExpressionMaterialFunctionCall_10, output=Result}
   ]
  }
  MaterialExpressionAdd_15 {
   class=MaterialExpressionAdd
   inputs [
    A {from=MaterialExpressionMaterialFunctionCall_22, output=Result}
    B {from=""}
   ]
   properties {constB=0.01}
  }
  MaterialExpressionAdd_16 {
   class=MaterialExpressionAdd
   inputs [
    A {from=MaterialExpressionNamedRerouteDeclaration_13}
    B {from=MaterialExpressionCustom_3}
   ]
  }
  MaterialExpressionAdd_17 {
   class=MaterialExpressionAdd
   inputs [
    A {from=MaterialExpressionMultiply_41}
    B {from=MaterialExpressionAdd_3}
   ]
  }
  MaterialExpressionAdd_2 {
   class=MaterialExpressionAdd
   inputs [
    A {from=MaterialExpressionMultiply_13}
    B {from=MaterialExpressionMultiply_4}
   ]
   properties {constB=0}
  }
  MaterialExpressionAdd_3 {
   class=MaterialExpressionAdd
   inputs [
    A {from=MaterialExpressionReroute_1}
    B {from=MaterialExpressionAdd_16}
   ]
  }
  MaterialExpressionAdd_4 {
   class=MaterialExpressionAdd
   inputs [
    A {from=MaterialExpressionCustom_7}
    B {from=MaterialExpressionMultiply_32}
   ]
  }
  MaterialExpressionAdd_5 {
   class=MaterialExpressionAdd
   inputs [
    A {from=MaterialExpressionNamedRerouteDeclaration_8}
    B {from=MaterialExpressionReroute_2}
   ]
  }
  MaterialExpressionAdd_6 {
   class=MaterialExpressionAdd
   inputs [
    A {from=MaterialExpressionScalarParameter_1}
    B {from=""}
   ]
  }
  MaterialExpressionAdd_7 {
   class=MaterialExpressionAdd
   inputs [
    A {from=MaterialExpressionMultiply_26}
    B {from=MaterialExpressionCustom_1, output=SwirlCenterBase_out}
   ]
  }
  MaterialExpressionAdd_8 {
   class=MaterialExpressionAdd
   inputs [
    A {from=MaterialExpressionSubtract_8}
    B {from=MaterialExpressionNamedRerouteUsage_37}
   ]
  }
  MaterialExpressionAdd_9 {
   class=MaterialExpressionAdd
   inputs [
    A {from=MaterialExpressionSine_0}
    B {from=MaterialExpressionConstant_9}
   ]
   properties {constB=0.5}
  }
  MaterialExpressionCameraPositionWS_0 {class=MaterialExpressionCameraPositionWS, inputs [], properties {}}
  MaterialExpressionClamp_0 {
   class=MaterialExpressionClamp
   inputs [
    None {from=MaterialExpressionAdd_4, output=""}
    Min {from="", output=""}
    Max {from="", output=""}
   ]
   properties {}
  }
  MaterialExpressionClamp_1 {
   class=MaterialExpressionClamp
   inputs [
    None {from=MaterialExpressionMaterialFunctionCall_25, output=Result}
    Min {from="", output=""}
    Max {from="", output=""}
   ]
   properties {}
  }
  MaterialExpressionClamp_2 {
   class=MaterialExpressionClamp
   inputs [
    None {from=MaterialExpressionMaterialFunctionCall_0, output=Result}
    Min {from=MaterialExpressionConstant_3, output=""}
    Max {from=MaterialExpressionMaterialFunctionCall_11, output=Result}
   ]
   properties {}
  }
  MaterialExpressionClamp_3 {
   class=MaterialExpressionClamp
   inputs [
    None {from=MaterialExpressionTextureSample_2, output=RGB}
    Min {from="", output=""}
    Max {from="", output=""}
   ]
   properties {minDefault=0.2}
  }
  MaterialExpressionClamp_4 {
   class=MaterialExpressionClamp
   inputs [
    None {from=MaterialExpressionMaterialFunctionCall_14, output=Result}
    Min {from="", output=""}
    Max {from="", output=""}
   ]
   properties {}
  }
  MaterialExpressionClamp_5 {
   class=MaterialExpressionClamp
   inputs [
    None {from=MaterialExpressionMaterialFunctionCall_29, output=Result}
    Min {from="", output=""}
    Max {from="", output=""}
   ]
   properties {}
  }
  MaterialExpressionClamp_6 {
   class=MaterialExpressionClamp
   inputs [
    None {from=MaterialExpressionMaterialFunctionCall_4, output=Result}
    Min {from="", output=""}
    Max {from="", output=""}
   ]
   properties {}
  }
  MaterialExpressionConstant2Vector_0 {class=MaterialExpressionConstant2Vector, properties {g=1, r=0.5}}
  MaterialExpressionConstant3Vector_0 {class=MaterialExpressionConstant3Vector, properties {constant=<rgba 0.7 0.8 1 0>}}
  MaterialExpressionConstant3Vector_1 {class=MaterialExpressionConstant3Vector, properties {constant=<rgba 1.1 0.8 0.7 0>}}
  MaterialExpressionConstant_0 {class=MaterialExpressionConstant, properties {r=500}}
  MaterialExpressionConstant_1 {class=MaterialExpressionConstant, properties {desc=dist from swirl to start puff, r=5}}
  MaterialExpressionConstant_10 {class=MaterialExpressionConstant, properties {desc=Sine Amp, r=3}}
  MaterialExpressionConstant_11 {class=MaterialExpressionConstant, properties {r=0.65}}
  MaterialExpressionConstant_12 {class=MaterialExpressionConstant, properties {r=0.6}}
  MaterialExpressionConstant_13 {class=MaterialExpressionConstant, properties {r=1}}
  MaterialExpressionConstant_14 {class=MaterialExpressionConstant, properties {r=300}}
  MaterialExpressionConstant_15 {class=MaterialExpressionConstant, properties {r=-25}}
  MaterialExpressionConstant_16 {class=MaterialExpressionConstant, properties {r=400}}
  MaterialExpressionConstant_17 {class=MaterialExpressionConstant, properties {r=0.25}}
  MaterialExpressionConstant_18 {class=MaterialExpressionConstant}
  MaterialExpressionConstant_19 {class=MaterialExpressionConstant, properties {r=1}}
  MaterialExpressionConstant_2 {class=MaterialExpressionConstant, properties {r=250}}
  MaterialExpressionConstant_20 {class=MaterialExpressionConstant, properties {r=1}}
  MaterialExpressionConstant_21 {class=MaterialExpressionConstant, properties {r=0.3}}
  MaterialExpressionConstant_22 {class=MaterialExpressionConstant, properties {r=0.6}}
  MaterialExpressionConstant_23 {class=MaterialExpressionConstant, properties {desc=corkscrew amt, r=0.3}}
  MaterialExpressionConstant_24 {class=MaterialExpressionConstant, properties {r=2}}
  MaterialExpressionConstant_25 {class=MaterialExpressionConstant, properties {r=5}}
  MaterialExpressionConstant_26 {class=MaterialExpressionConstant, properties {r=14}}
  MaterialExpressionConstant_27 {class=MaterialExpressionConstant, properties {r=100}}
  MaterialExpressionConstant_28 {class=MaterialExpressionConstant, properties {r=3}}
  MaterialExpressionConstant_29 {class=MaterialExpressionConstant, properties {r=0.01}}
  MaterialExpressionConstant_3 {class=MaterialExpressionConstant}
  MaterialExpressionConstant_30 {class=MaterialExpressionConstant, properties {r=0.045}}
  MaterialExpressionConstant_31 {class=MaterialExpressionConstant, properties {r=7}}
  MaterialExpressionConstant_32 {class=MaterialExpressionConstant, properties {desc=slowdown rot over this range, r=500}}
  MaterialExpressionConstant_33 {class=MaterialExpressionConstant, properties {desc=BaseRotSpeed, r=0.0015}}
  MaterialExpressionConstant_34 {class=MaterialExpressionConstant, properties {r=14}}
  MaterialExpressionConstant_35 {class=MaterialExpressionConstant, properties {r=12}}
  MaterialExpressionConstant_36 {class=MaterialExpressionConstant, properties {r=100}}
  MaterialExpressionConstant_37 {class=MaterialExpressionConstant, properties {r=6}}
  MaterialExpressionConstant_38 {class=MaterialExpressionConstant, properties {r=0.6}}
  MaterialExpressionConstant_39 {class=MaterialExpressionConstant}
  MaterialExpressionConstant_4 {class=MaterialExpressionConstant, properties {r=25}}
  MaterialExpressionConstant_40 {class=MaterialExpressionConstant, properties {r=1}}
  MaterialExpressionConstant_41 {class=MaterialExpressionConstant, properties {r=0.85}}
  MaterialExpressionConstant_42 {class=MaterialExpressionConstant, properties {r=1}}
  MaterialExpressionConstant_43 {class=MaterialExpressionConstant}
  MaterialExpressionConstant_44 {class=MaterialExpressionConstant, properties {r=12}}
  MaterialExpressionConstant_45 {class=MaterialExpressionConstant, properties {r=15}}
  MaterialExpressionConstant_46 {class=MaterialExpressionConstant, properties {r=778}}
  MaterialExpressionConstant_47 {class=MaterialExpressionConstant, properties {r=0.1}}
  MaterialExpressionConstant_48 {class=MaterialExpressionConstant, properties {r=15}}
  MaterialExpressionConstant_49 {class=MaterialExpressionConstant, properties {r=10}}
  MaterialExpressionConstant_5 {class=MaterialExpressionConstant}
  MaterialExpressionConstant_50 {class=MaterialExpressionConstant, properties {r=0.1}}
  MaterialExpressionConstant_52 {class=MaterialExpressionConstant, properties {r=0.35}}
  MaterialExpressionConstant_53 {class=MaterialExpressionConstant, properties {r=100}}
  MaterialExpressionConstant_6 {class=MaterialExpressionConstant, properties {r=100}}
  MaterialExpressionConstant_7 {class=MaterialExpressionConstant}
  MaterialExpressionConstant_8 {class=MaterialExpressionConstant, properties {r=30}}
  MaterialExpressionConstant_9 {class=MaterialExpressionConstant, properties {desc="values over 1 will avoid \r\nline thinning to 0", r=1.5}}
  MaterialExpressionConvert_0 {
   class=MaterialExpressionConvert
   inputs [None {from=MaterialExpressionTextureCoordinate_1}]
   properties {
    convertInputs [{
      type=Vector2
      expressionInput {
       expression=&self:MaterialExpressionTextureCoordinate_1
       inputName=None
       mask=0
       maskA=0
       maskB=0
       maskG=0
       maskR=0
       outputIndex=0
      }
     }]
    convertOutputs [{type=Scalar}, {type=Scalar}]
   }
  }
  MaterialExpressionConvert_1 {
   class=MaterialExpressionConvert
   inputs [
    None {from=MaterialExpressionClamp_3}
    None {from=MaterialExpressionTextureSample_2, output=A}
   ]
   properties {
    convertInputs [
     {
      type=Vector3
      expressionInput {
       expression=&self:MaterialExpressionClamp_3
       inputName=None
       mask=0
       maskA=0
       maskB=0
       maskG=0
       maskR=0
       outputIndex=0
      }
     }
     {
      type=Scalar
      expressionInput {
       expression=&self:MaterialExpressionTextureSample_2
       inputName=None
       mask=1
       maskA=1
       maskB=0
       maskG=0
       maskR=0
       outputIndex=4
      }
     }
    ]
    convertOutputs [{type=Vector4, defaultValue=<rgba 0 0 0 1>}]
   }
  }
  MaterialExpressionConvert_2 {
   class=MaterialExpressionConvert
   inputs [None {from=MaterialExpressionPanner_1}]
   properties {
    convertInputs [{
      type=Vector2
      expressionInput {
       expression=&self:MaterialExpressionPanner_1
       inputName=None
       mask=0
       maskA=0
       maskB=0
       maskG=0
       maskR=0
       outputIndex=0
      }
     }]
    convertOutputs [{type=Scalar}]
   }
  }
  MaterialExpressionConvert_3 {
   class=MaterialExpressionConvert
   inputs [None {from=MaterialExpressionTextureCoordinate_2}]
   properties {
    convertInputs [{
      type=Vector2
      expressionInput {
       expression=&self:MaterialExpressionTextureCoordinate_2
       inputName=None
       mask=0
       maskA=0
       maskB=0
       maskG=0
       maskR=0
       outputIndex=0
      }
     }]
    convertOutputs [{type=Scalar}]
   }
  }
  MaterialExpressionCustom_0 {
   class=MaterialExpressionCustom
   inputs [
    pos0 {from=MaterialExpressionNamedRerouteDeclaration_7, output=""}
    pos1 {from=MaterialExpressionNamedRerouteDeclaration_6, output=""}
    pos2 {from=MaterialExpressionNamedRerouteDeclaration_11, output=""}
    pos3 {from=MaterialExpressionNamedRerouteDeclaration_9, output=""}
    pos4 {from=MaterialExpressionNamedRerouteUsage_6, output=""}
   ]
   properties {
    code="return pos0 + pos1 + pos2 + pos3 + pos4;"
    description=AddPos
    showCode=true
    inputs [
     {
      inputName=pos0
      input {
       expression=&self:MaterialExpressionNamedRerouteDeclaration_7
       inputName=None
       mask=0
       maskA=0
       maskB=0
       maskG=0
       maskR=0
       outputIndex=0
      }
     }
     {
      inputName=pos1
      input {
       expression=&self:MaterialExpressionNamedRerouteDeclaration_6
       inputName=None
       mask=0
       maskA=0
       maskB=0
       maskG=0
       maskR=0
       outputIndex=0
      }
     }
     {
      inputName=pos2
      input {
       expression=&self:MaterialExpressionNamedRerouteDeclaration_11
       inputName=None
       mask=0
       maskA=0
       maskB=0
       maskG=0
       maskR=0
       outputIndex=0
      }
     }
     {
      inputName=pos3
      input {
       expression=&self:MaterialExpressionNamedRerouteDeclaration_9
       inputName=None
       mask=0
       maskA=0
       maskB=0
       maskG=0
       maskR=0
       outputIndex=0
      }
     }
     {
      inputName=pos4
      input {
       expression=&self:MaterialExpressionNamedRerouteUsage_6
       inputName=None
       mask=0
       maskA=0
       maskB=0
       maskG=0
       maskR=0
       outputIndex=0
      }
     }
    ]
   }
  }
  MaterialExpressionCustom_1 {
   class=MaterialExpressionCustom
   inputs [
    pos {from=MaterialExpressionDynamicParameter_0, output=RGB}
    orig_pos {from=MaterialExpressionDynamicParameter_1, output=RGB}
    slowdown_dist_start {from=MaterialExpressionMaterialFunctionCall_13, output=Result}
    slowdown_amt {from=MaterialExpressionMaterialFunctionCall_18, output=Result}
    radius_init {from=MaterialExpressionMaterialFunctionCall_20, output=Result}
    radius_drift {from=MaterialExpressionMaterialFunctionCall_23, output=Result}
    FaceCamera {from=MaterialExpressionStaticSwitchParameter_2, output=""}
    CameraVec {from=MaterialExpressionIf_0, output=""}
   ]
   properties {
    code="float start = slowdown_dist_start;\r\n\r\n//**Distance to use for swirl\r\nfloat3 pos_adj = pos - orig_pos;\r\nfloat dist = length(pos_adj);\r\n\r\n//**When the vtx go past swirl_dist they will curl\r\nfloat dist_swirl_amt = max(dist - start, 0);\r\ndist_swirl_amt *= slowdown_amt;\r\nfloat swirl_dist = dist - dist_swirl_amt;\r\nSwirlDist_out = swirl_dist;\r\n\r\n//**Unit vec in dir of swirl center\r\nfloat3 up_vec = 0, perp_axis = 0, radius_vec=0;\r\nif(FaceCamera==1) {\r\n    up_vec = normalize(CameraVec); \r\n    radius_vec = normalize(cross(pos_adj, up_vec));   \r\n    perp_axis = normalize(cross(pos_adj, radius_vec));\r\n}\r\nelse{\r\n    up_vec = normalize(float3(0, 1, 0.0001f));\r\n    perp_axis = normalize(cross(pos_adj, up_vec));\r\n    radius_vec = normalize(cross(perp_axis, pos_adj));\r\n}\r\n\r\n//**Position of center of the swirl\r\nSwirlCenterBase_out = normalize(pos_adj) * swirl_dist;\r\nfloat3 drift_vec = radius_vec * dist_swirl_amt * radius_drift;\r\n//prevent swirl getting to big because \r\ndrift_vec *=  1/slowdown_amt;\r\nfloat3 rad_offset_vec = (radius_vec * radius_init) + drift_vec;  \r\n\r\n//**Swirl Vector\r\nSwirlAxis_out = perp_axis;\r\nSwirlRadVec_out = rad_offset_vec;\r\n\r\nreturn 1;"
    description=Swirl Calcs
    showCode=true
    additionalOutputs [
     {outputName=SwirlCenterBase_out, outputType=CMOT_Float3}
     {outputName=SwirlDist_out, outputType=CMOT_Float1}
     {outputName=SwirlAxis_out, outputType=CMOT_Float3}
     {outputName=SwirlRadVec_out, outputType=CMOT_Float3}
    ]
    inputs [
     {
      inputName=pos
      input {
       expression=&self:MaterialExpressionDynamicParameter_0
       inputName=None
       mask=1
       maskA=0
       maskB=1
       maskG=1
       maskR=1
       outputIndex=4
      }
     }
     {
      inputName=orig_pos
      input {
       expression=&self:MaterialExpressionDynamicParameter_1
       inputName=None
       mask=1
       maskA=0
       maskB=1
       maskG=1
       maskR=1
       outputIndex=4
      }
     }
     {
      inputName=slowdown_dist_start
      input {
       expression=&self:MaterialExpressionMaterialFunctionCall_13
       inputName=None
       mask=0
       maskA=0
       maskB=0
       maskG=0
       maskR=0
       outputIndex=0
      }
     }
     {
      inputName=slowdown_amt
      input {
       expression=&self:MaterialExpressionMaterialFunctionCall_18
       inputName=None
       mask=0
       maskA=0
       maskB=0
       maskG=0
       maskR=0
       outputIndex=0
      }
     }
     {
      inputName=radius_init
      input {
       expression=&self:MaterialExpressionMaterialFunctionCall_20
       inputName=None
       mask=0
       maskA=0
       maskB=0
       maskG=0
       maskR=0
       outputIndex=0
      }
     }
     {
      inputName=radius_drift
      input {
       expression=&self:MaterialExpressionMaterialFunctionCall_23
       inputName=None
       mask=0
       maskA=0
       maskB=0
       maskG=0
       maskR=0
       outputIndex=0
      }
     }
     {
      inputName=FaceCamera
      input {
       expression=&self:MaterialExpressionStaticSwitchParameter_2
       inputName=None
       mask=0
       maskA=0
       maskB=0
       maskG=0
       maskR=0
       outputIndex=0
      }
     }
     {
      inputName=CameraVec
      input {
       expression=&self:MaterialExpressionIf_0
       inputName=None
       mask=0
       maskA=0
       maskB=0
       maskG=0
       maskR=0
       outputIndex=0
      }
     }
    ]
   }
  }
  MaterialExpressionCustom_2 {
   class=MaterialExpressionCustom
   inputs [
    rot_amt {from=MaterialExpressionNamedRerouteUsage_50, output=""}
    rad_vec {from=MaterialExpressionMultiply_27, output=""}
    swirl_in_amt {from=MaterialExpressionConstant_0, output=""}
    axis {from=MaterialExpressionNamedRerouteUsage_27, output=""}
    poke_out_amt {from=MaterialExpressionConstant_2, output=""}
    u {from=MaterialExpressionConvert_3, output=""}
    min_roll_size {from=MaterialExpressionConstant_50, output=""}
   ]
   properties {
    code="float3 roll = rad_vec * rot_amt * swirl_in_amt * .001;\r\nfloat mrs = min_roll_size;\r\nroll = normalize(rad_vec) * min(length(roll), ((1-u)*mrs+(1-mrs)) * length(rad_vec));\r\n\r\nfloat3 extrude = axis * length(roll) * poke_out_amt * (1/swirl_in_amt);\r\n\r\nreturn roll + extrude;"
    description=swirl_roll
    showCode=true
    inputs [
     {
      inputName=rot_amt
      input {
       expression=&self:MaterialExpressionNamedRerouteUsage_50
       inputName=None
       mask=0
       maskA=0
       maskB=0
       maskG=0
       maskR=0
       outputIndex=0
      }
     }
     {
      inputName=rad_vec
      input {
       expression=&self:MaterialExpressionMultiply_27
       inputName=None
       mask=0
       maskA=0
       maskB=0
       maskG=0
       maskR=0
       outputIndex=0
      }
     }
     {
      inputName=swirl_in_amt
      input {
       expression=&self:MaterialExpressionConstant_0
       inputName=None
       mask=0
       maskA=0
       maskB=0
       maskG=0
       maskR=0
       outputIndex=0
      }
     }
     {
      inputName=axis
      input {
       expression=&self:MaterialExpressionNamedRerouteUsage_27
       inputName=None
       mask=0
       maskA=0
       maskB=0
       maskG=0
       maskR=0
       outputIndex=0
      }
     }
     {
      inputName=poke_out_amt
      input {
       expression=&self:MaterialExpressionConstant_2
       inputName=None
       mask=0
       maskA=0
       maskB=0
       maskG=0
       maskR=0
       outputIndex=0
      }
     }
     {
      inputName=u
      input {
       expression=&self:MaterialExpressionConvert_3
       inputName=None
       mask=0
       maskA=0
       maskB=0
       maskG=0
       maskR=0
       outputIndex=0
      }
     }
     {
      inputName=min_roll_size
      input {
       expression=&self:MaterialExpressionConstant_50
       inputName=None
       mask=0
       maskA=0
       maskB=0
       maskG=0
       maskR=0
       outputIndex=0
      }
     }
    ]
   }
  }
  MaterialExpressionCustom_3 {
   class=MaterialExpressionCustom
   inputs [
    vtx {from=MaterialExpressionReroute_8, output=""}
    min_dist {from=MaterialExpressionConstant_4, output=""}
    max_dist {from=MaterialExpressionConstant_14, output=""}
    dir {from=MaterialExpressionNamedRerouteUsage_20, output=""}
    amt {from=MaterialExpressionConstant_11, output=""}
   ]
   properties {
    code="float3 up = float3(0,0,1);\r\n\r\nfloat3 n_dir = normalize(dir);\r\nfloat3 dir_proj = n_dir * dot(n_dir, vtx); //Project vertex onto dir axis\r\nfloat3 up_proj  = up * dot(dir_proj, up); //Project onto up vec\r\nfloat3 perp = -(dir_proj - up_proj); //Get vector to dir_proj\r\n\r\n//normalize min max range to get curve\r\nfloat len = max(0, length(dir_proj) - min_dist);\r\nfloat range = max_dist - min_dist;\r\nlen = min(len, range);\r\nfloat n_dist = len/range;\r\nn_dist *= amt;\r\n\r\n//Retrun the wpo\r\nreturn perp * n_dist;\r\n\r\n\r\n"
    description=SlowBendUp
    showCode=true
    inputs [
     {
      inputName=vtx
      input {
       expression=&self:MaterialExpressionReroute_8
       inputName=None
       mask=0
       maskA=0
       maskB=0
       maskG=0
       maskR=0
       outputIndex=0
      }
     }
     {
      inputName=min_dist
      input {
       expression=&self:MaterialExpressionConstant_4
       inputName=None
       mask=0
       maskA=0
       maskB=0
       maskG=0
       maskR=0
       outputIndex=0
      }
     }
     {
      inputName=max_dist
      input {
       expression=&self:MaterialExpressionConstant_14
       inputName=None
       mask=0
       maskA=0
       maskB=0
       maskG=0
       maskR=0
       outputIndex=0
      }
     }
     {
      inputName=dir
      input {
       expression=&self:MaterialExpressionNamedRerouteUsage_20
       inputName=None
       mask=0
       maskA=0
       maskB=0
       maskG=0
       maskR=0
       outputIndex=0
      }
     }
     {
      inputName=amt
      input {
       expression=&self:MaterialExpressionConstant_11
       inputName=None
       mask=0
       maskA=0
       maskB=0
       maskG=0
       maskR=0
       outputIndex=0
      }
     }
    ]
   }
  }
  MaterialExpressionCustom_4 {
   class=MaterialExpressionCustom
   inputs [
    amt {from=MaterialExpressionConstant_17, output=""}
    center {from=MaterialExpressionSubtract_10, output=""}
    up {from=MaterialExpressionNamedRerouteUsage_21, output=""}
    rot_wpo {from=MaterialExpressionMaterialFunctionCall_7, output=WPO}
   ]
   properties {
    code="//Normalize vectors\r\nfloat3 loc = rot_wpo-center;\r\nfloat3 n_loc = normalize(loc);\r\nfloat dist_loc = length(loc);\r\n\r\n//Streatch from swirl center\r\nfloat3 n_up = normalize(up);\r\nfloat ratio = dot(n_up, n_loc);\r\nfloat3 streatch = n_up * ratio * dist_loc * amt;\r\n\r\n//Squash back towards axis (not swirl center)\r\nfloat3 n_perp = normalize(center);\r\nfloat3 perp_vec = n_perp * dot(n_perp, rot_wpo);\r\nfloat3 squash = -perp_vec * min(amt, .25);\r\n\r\n\r\nreturn (loc + streatch + center) + squash;\r\n"
    desc="Can't figure out why the squash scales outer edge to nothing???\r\nThe bug actually looks good so leaving it in.\r\nAdd a min() fix to get by for now.\r\nEven an overall scale coming into this node creates the thinning."
    description=StreatchSwirlWpo
    showCode=true
    inputs [
     {
      inputName=amt
      input {
       expression=&self:MaterialExpressionConstant_17
       inputName=None
       mask=0
       maskA=0
       maskB=0
       maskG=0
       maskR=0
       outputIndex=0
      }
     }
     {
      inputName=center
      input {
       expression=&self:MaterialExpressionSubtract_10
       inputName=None
       mask=0
       maskA=0
       maskB=0
       maskG=0
       maskR=0
       outputIndex=0
      }
     }
     {
      inputName=up
      input {
       expression=&self:MaterialExpressionNamedRerouteUsage_21
       inputName=None
       mask=0
       maskA=0
       maskB=0
       maskG=0
       maskR=0
       outputIndex=0
      }
     }
     {
      inputName=rot_wpo
      input {
       expression=&self:MaterialExpressionMaterialFunctionCall_7
       inputName=None
       mask=0
       maskA=0
       maskB=0
       maskG=0
       maskR=0
       outputIndex=0
      }
     }
    ]
   }
  }
  MaterialExpressionCustom_5 {
   class=MaterialExpressionCustom
   inputs [
    v {from=MaterialExpressionTextureCoordinate_4, output=""}
    Intensity {from=MaterialExpressionConstant_52, output=""}
    Splinethicken {from=MaterialExpressionNamedRerouteUsage_23, output=""}
   ]
   properties {
    code="float v1 = min(v.y*2, 1);\r\nfloat v2 = min((1-v.y)*2, 1);\r\nfloat result = v1*v2*Intensity*Splinethicken;\r\n\r\nreturn result;"
    outputType=CMOT_Float1
    showCode=true
    inputs [
     {
      inputName=v
      input {
       expression=&self:MaterialExpressionTextureCoordinate_4
       inputName=None
       mask=0
       maskA=0
       maskB=0
       maskG=0
       maskR=0
       outputIndex=0
      }
     }
     {
      inputName=Intensity
      input {
       expression=&self:MaterialExpressionConstant_52
       inputName=None
       mask=0
       maskA=0
       maskB=0
       maskG=0
       maskR=0
       outputIndex=0
      }
     }
     {
      inputName=Splinethicken
      input {
       expression=&self:MaterialExpressionNamedRerouteUsage_23
       inputName=None
       mask=0
       maskA=0
       maskB=0
       maskG=0
       maskR=0
       outputIndex=0
      }
     }
    ]
   }
  }
  MaterialExpressionCustom_6 {
   class=MaterialExpressionCustom
   inputs [
    tex_coord {from=MaterialExpressionTextureCoordinate_0, output=""}
    pos {from=MaterialExpressionAdd_0, output=""}
    world_base_taper {from=MaterialExpressionConstant_8, output=""}
   ]
   properties {
    code="//Taper back and front ends of the mesh\r\nfloat taper_u= .08;\r\nfloat taper_out = clamp(tex_coord.x/taper_u, .2, 1);\r\nfloat taper_in = clamp((1-tex_coord.x)/taper_u, .2, 1);\r\n\r\n//Taper from origin\r\nfloat in_dist = world_base_taper;\r\nfloat dist = length(pos);\r\nfloat taper_0 = clamp(dist/in_dist, 0, 1);\r\n\r\n//Thicken towards swirl\r\n\r\n\r\nreturn taper_out * taper_in * taper_0;"
    desc="taper base, tip, and from emission point"
    description=taper
    outputType=CMOT_Float1
    showCode=true
    inputs [
     {
      inputName=tex_coord
      input {
       expression=&self:MaterialExpressionTextureCoordinate_0
       inputName=None
       mask=0
       maskA=0
       maskB=0
       maskG=0
       maskR=0
       outputIndex=0
      }
     }
     {
      inputName=pos
      input {
       expression=&self:MaterialExpressionAdd_0
       inputName=None
       mask=0
       maskA=0
       maskB=0
       maskG=0
       maskR=0
       outputIndex=0
      }
     }
     {
      inputName=world_base_taper
      input {
       expression=&self:MaterialExpressionConstant_8
       inputName=None
       mask=0
       maskA=0
       maskB=0
       maskG=0
       maskR=0
       outputIndex=0
      }
     }
    ]
   }
  }
  MaterialExpressionCustom_7 {
   class=MaterialExpressionCustom
   inputs [Alpha {from=MaterialExpressionConvert_0, output=""}]
   properties {
    code="float4 c1 = {0,0,0,1};\r\nfloat4 c2 = {.1, .5, .9,1};\r\nreturn lerp(c1, c2, Alpha); "
    outputType=CMOT_Float4
    inputs [{
      inputName=Alpha
      input {
       expression=&self:MaterialExpressionConvert_0
       inputName=None
       mask=0
       maskA=0
       maskB=0
       maskG=0
       maskR=0
       outputIndex=0
      }
     }]
   }
  }
  MaterialExpressionCustom_8 {
   class=MaterialExpressionCustom
   inputs [Alpha {from=MaterialExpressionConvert_0, output=""}]
   properties {
    code="float4 c1 = {1,.09,.7,1};\r\nfloat4 c2 = {1, .6, .14,1};\r\nreturn lerp(c1, c2, Alpha); "
    outputType=CMOT_Float4
    inputs [{
      inputName=Alpha
      input {
       expression=&self:MaterialExpressionConvert_0
       inputName=None
       mask=0
       maskA=0
       maskB=0
       maskG=0
       maskR=0
       outputIndex=1
      }
     }]
   }
  }
  MaterialExpressionDivide_5 {
   class=MaterialExpressionDivide
   inputs [
    A {from=MaterialExpressionMultiply_2}
    B {from=MaterialExpressionNamedRerouteUsage_45}
   ]
   properties {desc=Maintian rot no matter speed of swirl}
  }
  MaterialExpressionDotProduct_0 {
   class=MaterialExpressionDotProduct
   inputs [
    A {from=MaterialExpressionNamedRerouteUsage_3, output=""}
    B {from=MaterialExpressionNamedRerouteUsage_7, output=""}
   ]
   properties {}
  }
  MaterialExpressionDotProduct_1 {
   class=MaterialExpressionDotProduct
   inputs [
    A {from=MaterialExpressionNormalize_0, output=""}
    B {from=MaterialExpressionMultiply_0, output=""}
   ]
   properties {desc=perp dist}
  }
  MaterialExpressionDynamicParameter_0 {class=MaterialExpressionDynamicParameter, inputs [], properties {defaultValue=<rgba 1 0.9428 0.0132 1>, paramNames [posx,poxy,posz,None]}}
  MaterialExpressionDynamicParameter_1 {
   class=MaterialExpressionDynamicParameter
   inputs []
   properties {parameterIndex=1, defaultValue=<rgba 1 0.9428 0.0132 1>, paramNames [origposx,origpoxy,origposz,None]}
  }
  MaterialExpressionIf_0 {
   class=MaterialExpressionIf
   inputs [
    A {from=MaterialExpressionMaterialFunctionCall_15, output=Result}
    B {from="", output=""}
    "A > B" {from=MaterialExpressionSubtract_11, output=""}
    "A == B" {from=MaterialExpressionSubtract_11, output=""}
    "A < B" {from=MaterialExpressionMaterialFunctionCall_12, output=Result}
   ]
   properties {}
  }
  MaterialExpressionIf_1 {
   class=MaterialExpressionIf
   inputs [
    A {from=MaterialExpressionMaterialFunctionCall_8, output=Result}
    B {from=MaterialExpressionConstant_23, output=""}
    "A > B" {from=MaterialExpressionReroute_5, output=""}
    "A == B" {from=MaterialExpressionAdd_12, output=""}
    "A < B" {from=MaterialExpressionAdd_12, output=""}
   ]
   properties {constB=0.3}
  }
  MaterialExpressionLength_0 {class=MaterialExpressionLength, inputs [None {from=MaterialExpressionNamedRerouteUsage_10, output=""}], properties {}}
  MaterialExpressionLength_1 {class=MaterialExpressionLength, inputs [None {from=MaterialExpressionNamedRerouteUsage_17, output=""}], properties {}}
  MaterialExpressionLength_3 {class=MaterialExpressionLength, inputs [None {from=MaterialExpressionReroute_9, output=""}], properties {desc=WorldSpaceLengthToEachVtx}}
  MaterialExpressionLinearInterpolate_0 {
   class=MaterialExpressionLinearInterpolate
   inputs [
    A {from=""}
    B {from=MaterialExpressionSubtract_1}
    Alpha {from=MaterialExpressionScalarParameter_4}
   ]
   properties {constA=1}
  }
  MaterialExpressionMaterialFunctionCall_0 {
   class=MaterialExpressionMaterialFunctionCall
   inputs [
    Input {from=MaterialExpressionLength_1}
    Input Low {from=MaterialExpressionSubtract_0}
    Input High {from=MaterialExpressionAdd_1}
    Target Low {from=MaterialExpressionConstant_3}
    Target High {from=MaterialExpressionMaterialFunctionCall_11, output=Result}
   ]
   properties {materialFunction=&RemapValueRange}
  }
  MaterialExpressionMaterialFunctionCall_1 {
   class=MaterialExpressionMaterialFunctionCall
   inputs [
    Input {from=MaterialExpressionParticleRandom_1}
    Input Low {from=""}
    Input High {from=""}
    Target Low {from=MaterialExpressionAdd_6}
    Target High {from=""}
   ]
   properties {materialFunction=&RemapValueRange}
  }
  MaterialExpressionMaterialFunctionCall_10 {
   class=MaterialExpressionMaterialFunctionCall
   inputs [
    Input {from=MaterialExpressionParticleRandom_8}
    min {from=""}
    max {from=MaterialExpressionConstant_36}
    seed {from=MaterialExpressionConstant_35}
   ]
   properties {materialFunction=/InterchangeAssets/Functions/MX_RandomFloat}
  }
  MaterialExpressionMaterialFunctionCall_11 {
   class=MaterialExpressionMaterialFunctionCall
   inputs [
    Input {from=MaterialExpressionParticleRandom_9}
    min {from=MaterialExpressionConstant_37}
    max {from=MaterialExpressionConstant_44}
    seed {from=MaterialExpressionConstant_34}
   ]
   properties {materialFunction=&MX_RandomFloat}
  }
  MaterialExpressionMaterialFunctionCall_12 {
   class=MaterialExpressionMaterialFunctionCall
   inputs []
   properties {materialFunction=&CameraDirectionVector}
  }
  MaterialExpressionMaterialFunctionCall_13 {
   class=MaterialExpressionMaterialFunctionCall
   inputs [
    Input {from=MaterialExpressionParticleRandom_3}
    min {from=MaterialExpressionConstant_15}
    max {from=MaterialExpressionConstant_16}
    seed {from=MaterialExpressionConstant_7}
   ]
   properties {materialFunction=&MX_RandomFloat}
  }
  MaterialExpressionMaterialFunctionCall_14 {
   class=MaterialExpressionMaterialFunctionCall
   inputs [
    Input {from=MaterialExpressionTextureSample_1, output=R}
    Input Low {from=MaterialExpressionMaterialFunctionCall_22, output=Result}
    Input High {from=MaterialExpressionAdd_15}
    Target Low {from=""}
    Target High {from=""}
   ]
   properties {materialFunction=&RemapValueRange}
  }
  MaterialExpressionMaterialFunctionCall_15 {
   class=MaterialExpressionMaterialFunctionCall
   inputs [
    Input {from=MaterialExpressionParticleRandom_2}
    min {from=""}
    max {from=""}
    seed {from=MaterialExpressionConstant_46}
   ]
   properties {materialFunction=&MX_RandomFloat}
  }
  MaterialExpressionMaterialFunctionCall_16 {
   class=MaterialExpressionMaterialFunctionCall
   inputs [
    Input {from=MaterialExpressionParticleRandom_10}
    min {from=MaterialExpressionMultiply_28}
    max {from=MaterialExpressionReroute_7}
    seed {from=MaterialExpressionConstant_45}
   ]
   properties {materialFunction=&MX_RandomFloat}
  }
  MaterialExpressionMaterialFunctionCall_17 {
   class=MaterialExpressionMaterialFunctionCall
   inputs [
    Fade Length {from=MaterialExpressionConstant_27}
    Fade Offset {from=MaterialExpressionConstant_49}
    For Vertex Shader {from=""}
   ]
   properties {materialFunction=&CameraDepthFade}
  }
  MaterialExpressionMaterialFunctionCall_18 {
   class=MaterialExpressionMaterialFunctionCall
   inputs [
    Input {from=MaterialExpressionParticleRandom_5}
    min {from=MaterialExpressionConstant_21}
    max {from=MaterialExpressionConstant_22}
    seed {from=MaterialExpressionConstant_20}
   ]
   properties {materialFunction=&MX_RandomFloat}
  }
  MaterialExpressionMaterialFunctionCall_2 {
   class=MaterialExpressionMaterialFunctionCall
   inputs [
    WidthBase {from=MaterialExpressionReroute_4}
    WidthTip {from=MaterialExpressionReroute_4}
    WorldPosition {from=""}
    UVs for Projection {from=""}
    UVs for Thickness {from=""}
    UVs for texturing {from=""}
    Expand U or V UV Channel {from=MaterialExpressionStaticBool_0}
    DeriveNormalZ {from=""}
    AngleCorrectedNormals {from=""}
    AdditionalNormal {from=MaterialExpressionMaterialFunctionCall_24, output=Result}
    FlattenPixelNormal {from=MaterialExpressionStaticBool_1}
    UVs For NormalShading {from=""}
   ]
   properties {materialFunction=&SplineThicken}
  }
  MaterialExpressionMaterialFunctionCall_20 {
   class=MaterialExpressionMaterialFunctionCall
   inputs [
    Input {from=MaterialExpressionParticleRandom_6}
    min {from=MaterialExpressionConstant_25}
    max {from=MaterialExpressionConstant_26}
    seed {from=MaterialExpressionConstant_24}
   ]
   properties {materialFunction=&MX_RandomFloat}
  }
  MaterialExpressionMaterialFunctionCall_22 {
   class=MaterialExpressionMaterialFunctionCall
   inputs [
    Input {from=MaterialExpressionNamedRerouteUsage_43}
    Input Low {from=MaterialExpressionConstant_38}
    Input High {from=""}
    Target Low {from=""}
    Target High {from=""}
   ]
   properties {materialFunction=&RemapValueRange}
  }
  MaterialExpressionMaterialFunctionCall_23 {
   class=MaterialExpressionMaterialFunctionCall
   inputs [
    Input {from=MaterialExpressionParticleRandom_7}
    min {from=MaterialExpressionConstant_29}
    max {from=MaterialExpressionConstant_30}
    seed {from=MaterialExpressionConstant_28}
   ]
   properties {materialFunction=&MX_RandomFloat}
  }
  MaterialExpressionMaterialFunctionCall_24 {
   class=MaterialExpressionMaterialFunctionCall
   inputs [
    Height Map {from=MaterialExpressionTextureObject_1}
    Normal Map Intensity {from=MaterialExpressionConstant_53}
    Height Map UV Offset {from=""}
    Coordinates {from=""}
    Height Map Channel Selector {from=""}
   ]
   properties {materialFunction=/Engine/Functions/Engine_MaterialFunctions03/Procedurals/NormalFromHeightmap}
  }
  MaterialExpressionMaterialFunctionCall_25 {
   class=MaterialExpressionMaterialFunctionCall
   inputs [
    Input {from=MaterialExpressionNamedRerouteUsage_44}
    Input Low {from=MaterialExpressionConstant_12}
    Input High {from=MaterialExpressionConstant_40}
    Target Low {from=MaterialExpressionConstant_13}
    Target High {from=MaterialExpressionConstant_39}
   ]
   properties {materialFunction=&RemapValueRange}
  }
  MaterialExpressionMaterialFunctionCall_26 {
   class=MaterialExpressionMaterialFunctionCall
   inputs [
    Hue Shift Percentage {from=MaterialExpressionMaterialFunctionCall_16, output=Result}
    Texture {from=MaterialExpressionMultiply_36}
   ]
   properties {materialFunction=&HueShift}
  }
  MaterialExpressionMaterialFunctionCall_27 {
   class=MaterialExpressionMaterialFunctionCall
   inputs [
    Input {from=MaterialExpressionParticleRandom_11}
    min {from=""}
    max {from=""}
    seed {from=MaterialExpressionConstant_48}
   ]
   properties {materialFunction=&MX_RandomFloat}
  }
  MaterialExpressionMaterialFunctionCall_29 {
   class=MaterialExpressionMaterialFunctionCall
   inputs [
    Input {from=MaterialExpressionNamedRerouteUsage_47}
    Input Low {from=MaterialExpressionConstant_41}
    Input High {from=""}
    Target Low {from=MaterialExpressionConstant_42}
    Target High {from=MaterialExpressionConstant_43}
   ]
   properties {materialFunction=&RemapValueRange}
  }
  MaterialExpressionMaterialFunctionCall_4 {
   class=MaterialExpressionMaterialFunctionCall
   inputs [
    Input {from=MaterialExpressionLength_0}
    Input Low {from=MaterialExpressionAdd_13}
    Input High {from=MaterialExpressionAdd_11}
    Target Low {from=""}
    Target High {from=""}
   ]
   properties {materialFunction=&RemapValueRange}
  }
  MaterialExpressionMaterialFunctionCall_5 {
   class=MaterialExpressionMaterialFunctionCall
   inputs [
    Input {from=MaterialExpressionReroute_3}
    Input Low {from=""}
    Input High {from=MaterialExpressionConstant_32}
    Target Low {from=MaterialExpressionMultiply_52}
    Target High {from=MaterialExpressionConstant_33}
   ]
   properties {materialFunction=&RemapValueRange}
  }
  MaterialExpressionMaterialFunctionCall_6 {
   class=MaterialExpressionMaterialFunctionCall
   inputs [
    Color_Morning {from=MaterialExpressionMultiply_20}
    Color_Day {from=MaterialExpressionReroute_11}
    Color_Evening {from=MaterialExpressionMultiply_20}
    Color_Night {from=MaterialExpressionMultiply_19}
   ]
   properties {materialFunction=&Color_Time_of_Day}
  }
  MaterialExpressionMaterialFunctionCall_7 {
   class=MaterialExpressionMaterialFunctionCall
   inputs [
    RotationAngle {from=MaterialExpressionReroute_0}
    NormalizedRotationAxis {from=MaterialExpressionMultiply_22}
    PivotPoint {from=MaterialExpressionNamedRerouteUsage_2}
    Position {from=MaterialExpressionReroute_2}
    Normal {from=MaterialExpressionNamedRerouteUsage_16}
    TangentU {from=""}
    TangentV {from=""}
   ]
   properties {materialFunction=/BRCosmetics/Characters/Player/Male/Medium/Bodies/M_MED_Inferno/FX/Mat/RotateVertexAboutAxis}
  }
  MaterialExpressionMaterialFunctionCall_8 {
   class=MaterialExpressionMaterialFunctionCall
   inputs [
    Input {from=MaterialExpressionParticleRandom_4}
    min {from=""}
    max {from=""}
    seed {from=MaterialExpressionConstant_31}
   ]
   properties {materialFunction=&MX_RandomFloat}
  }
  MaterialExpressionMaterialFunctionCall_9 {
   class=MaterialExpressionMaterialFunctionCall
   inputs [
    Input {from=MaterialExpressionLength_3}
    Input Low {from=MaterialExpressionConstant_5}
    Input High {from=MaterialExpressionConstant_6}
   ]
   properties {desc=return 0-1, materialFunction=&RemapValue_0-1}
  }
  MaterialExpressionMax_0 {
   class=MaterialExpressionMax
   inputs [
    A {from=MaterialExpressionSign_0}
    B {from=""}
   ]
   properties {constB=0}
  }
  MaterialExpressionMin_0 {
   class=MaterialExpressionMin
   inputs [
    A {from=MaterialExpressionDotProduct_1, output=""}
    B {from=MaterialExpressionNamedRerouteUsage_1, output=""}
   ]
   properties {}
  }
  MaterialExpressionMultiply_0 {
   class=MaterialExpressionMultiply
   inputs [
    A {from=MaterialExpressionNamedRerouteUsage_3}
    B {from=MaterialExpressionMax_0}
   ]
  }
  MaterialExpressionMultiply_1 {
   class=MaterialExpressionMultiply
   inputs [
    A {from=MaterialExpressionPower_0}
    B {from=""}
   ]
   properties {constB=20, desc=max height for opping on/off}
  }
  MaterialExpressionMultiply_10 {
   class=MaterialExpressionMultiply
   inputs [
    A {from=MaterialExpressionNormalize_0}
    B {from=MaterialExpressionSubtract_6}
   ]
   properties {desc=offset vector to flatten tip}
  }
  MaterialExpressionMultiply_11 {
   class=MaterialExpressionMultiply
   inputs [
    A {from=MaterialExpressionSine_1}
    B {from=MaterialExpressionMultiply_34}
   ]
  }
  MaterialExpressionMultiply_12 {
   class=MaterialExpressionMultiply
   inputs [
    A {from=MaterialExpressionAdd_9}
    B {from=MaterialExpressionConstant_10}
   ]
   properties {constB=15}
  }
  MaterialExpressionMultiply_13 {
   class=MaterialExpressionMultiply
   inputs [
    A {from=MaterialExpressionMultiply_12}
    B {from=MaterialExpressionCustom_6}
   ]
  }
  MaterialExpressionMultiply_14 {
   class=MaterialExpressionMultiply
   inputs [
    A {from=MaterialExpressionMultiply_10}
    B {from=""}
   ]
   properties {constB=-1}
  }
  MaterialExpressionMultiply_15 {
   class=MaterialExpressionMultiply
   inputs [
    A {from=MaterialExpressionTextureSample_2, output=A}
    B {from=MaterialExpressionMultiply_16}
   ]
  }
  MaterialExpressionMultiply_16 {
   class=MaterialExpressionMultiply
   inputs [
    A {from=MaterialExpressionMaterialFunctionCall_17, output=Result}
    B {from=MaterialExpressionMultiply_50}
   ]
  }
  MaterialExpressionMultiply_17 {
   class=MaterialExpressionMultiply
   inputs [
    A {from=MaterialExpressionNamedRerouteUsage_13}
    B {from=""}
   ]
   properties {constB=-1}
  }
  MaterialExpressionMultiply_18 {
   class=MaterialExpressionMultiply
   inputs [
    A {from=MaterialExpressionMaterialFunctionCall_6, output=Result}
    B {from=""}
   ]
   properties {constB=1.1}
  }
  MaterialExpressionMultiply_19 {
   class=MaterialExpressionMultiply
   inputs [
    A {from=MaterialExpressionReroute_11}
    B {from=MaterialExpressionConstant3Vector_0}
   ]
  }
  MaterialExpressionMultiply_2 {
   class=MaterialExpressionMultiply
   inputs [
    A {from=MaterialExpressionReroute_3}
    B {from=MaterialExpressionMaterialFunctionCall_5, output=Result}
   ]
   properties {constB=0.004}
  }
  MaterialExpressionMultiply_20 {
   class=MaterialExpressionMultiply
   inputs [
    A {from=MaterialExpressionConstant3Vector_1}
    B {from=MaterialExpressionReroute_11}
   ]
  }
  MaterialExpressionMultiply_21 {
   class=MaterialExpressionMultiply
   inputs [
    A {from=MaterialExpressionMultiply_23}
    B {from=MaterialExpressionMultiply_6}
   ]
   properties {constA=1}
  }
  MaterialExpressionMultiply_22 {
   class=MaterialExpressionMultiply
   inputs [
    A {from=MaterialExpressionNamedRerouteUsage_5}
    B {from=MaterialExpressionNamedRerouteUsage_30}
   ]
  }
  MaterialExpressionMultiply_23 {
   class=MaterialExpressionMultiply
   inputs [
    A {from=MaterialExpressionNormalize_1}
    B {from=MaterialExpressionLinearInterpolate_0}
   ]
  }
  MaterialExpressionMultiply_24 {
   class=MaterialExpressionMultiply
   inputs [
    A {from=MaterialExpressionRound_0}
    B {from=""}
   ]
   properties {constB=2}
  }
  MaterialExpressionMultiply_25 {
   class=MaterialExpressionMultiply
   inputs [
    A {from=MaterialExpressionMaterialFunctionCall_1, output=Result}
    B {from=MaterialExpressionMultiply_5}
   ]
  }
  MaterialExpressionMultiply_26 {
   class=MaterialExpressionMultiply
   inputs [
    A {from=MaterialExpressionNamedRerouteUsage_19}
    B {from=MaterialExpressionNamedRerouteUsage_22}
   ]
  }
  MaterialExpressionMultiply_27 {
   class=MaterialExpressionMultiply
   inputs [
    A {from=MaterialExpressionNamedRerouteUsage_15}
    B {from=MaterialExpressionNamedRerouteUsage_29}
   ]
  }
  MaterialExpressionMultiply_28 {
   class=MaterialExpressionMultiply
   inputs [
    A {from=MaterialExpressionConstant_47}
    B {from=""}
   ]
   properties {constB=-1}
  }
  MaterialExpressionMultiply_3 {
   class=MaterialExpressionMultiply
   inputs [
    A {from=MaterialExpressionRound_1}
    B {from=""}
   ]
   properties {constB=2}
  }
  MaterialExpressionMultiply_31 {
   class=MaterialExpressionMultiply
   inputs [
    A {from=MaterialExpressionScalarParameter_3}
    B {from=""}
   ]
   properties {constB=-1}
  }
  MaterialExpressionMultiply_32 {
   class=MaterialExpressionMultiply
   inputs [
    A {from=MaterialExpressionSubtract_9}
    B {from=MaterialExpressionCustom_8}
   ]
  }
  MaterialExpressionMultiply_34 {
   class=MaterialExpressionMultiply
   inputs [
    A {from=MaterialExpressionMaterialFunctionCall_9, output=Result}
    B {from=MaterialExpressionClamp_1}
   ]
  }
  MaterialExpressionMultiply_35 {
   class=MaterialExpressionMultiply
   inputs [
    A {from=MaterialExpressionMaterialFunctionCall_6, output=Result}
    B {from=""}
   ]
   properties {constB=0.45}
  }
  MaterialExpressionMultiply_36 {
   class=MaterialExpressionMultiply
   inputs [
    A {from=MaterialExpressionClamp_0}
    B {from=MaterialExpressionConvert_1}
   ]
  }
  MaterialExpressionMultiply_37 {
   class=MaterialExpressionMultiply
   inputs [
    A {from=MaterialExpressionAdd_14}
    B {from=""}
   ]
   properties {constB=0.01, desc=sine poppy freq}
  }
  MaterialExpressionMultiply_38 {
   class=MaterialExpressionMultiply
   inputs [
    A {from=MaterialExpressionNamedRerouteUsage_41}
    B {from=""}
   ]
   properties {constB=-0.5}
  }
  MaterialExpressionMultiply_39 {
   class=MaterialExpressionMultiply
   inputs [
    A {from=MaterialExpressionSine_2}
    B {from=""}
   ]
   properties {constB=0.5}
  }
  MaterialExpressionMultiply_4 {
   class=MaterialExpressionMultiply
   inputs [
    A {from=MaterialExpressionCustom_6}
    B {from=MaterialExpressionClamp_2}
   ]
  }
  MaterialExpressionMultiply_41 {
   class=MaterialExpressionMultiply
   inputs [
    A {from=MaterialExpressionNamedRerouteUsage_31}
    B {from=MaterialExpressionCustom_5}
   ]
  }
  MaterialExpressionMultiply_43 {
   class=MaterialExpressionMultiply
   inputs [
    A {from=MaterialExpressionAdd_2}
    B {from=MaterialExpressionScalarParameter_5}
   ]
  }
  MaterialExpressionMultiply_5 {
   class=MaterialExpressionMultiply
   inputs [
    A {from=MaterialExpressionLength_3}
    B {from=MaterialExpressionScalarParameter_2}
   ]
   properties {constB=0.002, desc=freq}
  }
  MaterialExpressionMultiply_50 {
   class=MaterialExpressionMultiply
   inputs [
    A {from=MaterialExpressionClamp_4}
    B {from=MaterialExpressionMultiply_51}
   ]
  }
  MaterialExpressionMultiply_51 {
   class=MaterialExpressionMultiply
   inputs [
    A {from=MaterialExpressionClamp_6}
    B {from=MaterialExpressionPower_1}
   ]
  }
  MaterialExpressionMultiply_52 {
   class=MaterialExpressionMultiply
   inputs [
    A {from=""}
    B {from=MaterialExpressionConstant_33}
   ]
   properties {constA=4, desc=InitRotMult}
  }
  MaterialExpressionMultiply_6 {
   class=MaterialExpressionMultiply
   inputs [
    A {from=MaterialExpressionMultiply_11}
    B {from=MaterialExpressionScalarParameter_6}
   ]
   properties {constB=100, desc=sin amt}
  }
  MaterialExpressionMultiply_7 {
   class=MaterialExpressionMultiply
   inputs [
    A {from=MaterialExpressionTextureCoordinate_0}
    B {from=""}
   ]
   properties {constB=1.5}
  }
  MaterialExpressionMultiply_8 {
   class=MaterialExpressionMultiply
   inputs [
    A {from=MaterialExpressionTextureCoordinate_1}
    B {from=MaterialExpressionNamedRerouteUsage_13}
   ]
  }
  MaterialExpressionMultiply_9 {
   class=MaterialExpressionMultiply
   inputs [
    A {from=MaterialExpressionMultiply_8}
    B {from=MaterialExpressionConstant2Vector_0}
   ]
  }
  MaterialExpressionNamedRerouteDeclaration_0 {
   class=MaterialExpressionNamedRerouteDeclaration
   inputs [None {from=MaterialExpressionMaterialFunctionCall_2, output=Normal}]
   properties {name=ThickenNormal, nodeColor=<rgba 0.2891 1 0 1>}
  }
  MaterialExpressionNamedRerouteDeclaration_1 {
   class=MaterialExpressionNamedRerouteDeclaration
   inputs [None {from=MaterialExpressionAdd_7}]
   properties {name=SwirlCenter, nodeColor=<rgba 1 0.8095 0 1>}
  }
  MaterialExpressionNamedRerouteDeclaration_10 {
   class=MaterialExpressionNamedRerouteDeclaration
   inputs [None {from=MaterialExpressionCustom_1, output=SwirlRadVec_out}]
   properties {name=SwirlRadVec, nodeColor=<rgba 1 0.8095 0 1>}
  }
  MaterialExpressionNamedRerouteDeclaration_11 {
   class=MaterialExpressionNamedRerouteDeclaration
   inputs [None {from=MaterialExpressionCustom_2}]
   properties {name=SwirlOffset_Roll, nodeColor=<rgba 1 0 0.7781 1>}
  }
  MaterialExpressionNamedRerouteDeclaration_12 {
   class=MaterialExpressionNamedRerouteDeclaration
   inputs [None {from=MaterialExpressionSubtract_5}]
   properties {name=SwirlPartLocalPos, nodeColor=<rgba 1 0.8705 0 1>}
  }
  MaterialExpressionNamedRerouteDeclaration_13 {
   class=MaterialExpressionNamedRerouteDeclaration
   inputs [None {from=MaterialExpressionMultiply_21}]
   properties {name=SwirlOffset_GlobalDistortion, nodeColor=<rgba 1 0 0.8008 1>}
  }
  MaterialExpressionNamedRerouteDeclaration_14 {
   class=MaterialExpressionNamedRerouteDeclaration
   inputs [None {from=MaterialExpressionMultiply_43}]
   properties {name=SwirlSplineThicken, nodeColor=<rgba 1 0 0.8967 1>}
  }
  MaterialExpressionNamedRerouteDeclaration_15 {
   class=MaterialExpressionNamedRerouteDeclaration
   inputs [None {from=MaterialExpressionScalarParameter_0}]
   properties {name=SwirlGlobalScale, nodeColor=<rgba 1 0.8434 0 1>}
  }
  MaterialExpressionNamedRerouteDeclaration_16 {
   class=MaterialExpressionNamedRerouteDeclaration
   inputs [None {from=MaterialExpressionSubtract_7}]
   properties {name=SwirlRadVecDoFlip, nodeColor=<rgba 1 0.9547 0 1>}
  }
  MaterialExpressionNamedRerouteDeclaration_17 {
   class=MaterialExpressionNamedRerouteDeclaration
   inputs [None {from=MaterialExpressionDynamicParameter_0, output=RGB}]
   properties {name=SwirlParticlePos, nodeColor=<rgba 1 0.7781 0 1>}
  }
  MaterialExpressionNamedRerouteDeclaration_18 {
   class=MaterialExpressionNamedRerouteDeclaration
   inputs [None {from=MaterialExpressionCustom_1, output=SwirlCenterBase_out}]
   properties {name=SwirlCenterBase, nodeColor=<rgba 1 0.9189 0 1>}
  }
  MaterialExpressionNamedRerouteDeclaration_19 {
   class=MaterialExpressionNamedRerouteDeclaration
   inputs [None {from=MaterialExpressionMaterialFunctionCall_7, output=Normal}]
   properties {name=SwirlNormalRotate, nodeColor=<rgba 0.5703 0 1 1>}
  }
  MaterialExpressionNamedRerouteDeclaration_2 {
   class=MaterialExpressionNamedRerouteDeclaration
   inputs [None {from=MaterialExpressionCustom_1, output=SwirlDist_out}]
   properties {name=SwirlDist, nodeColor=<rgba 1 0.8095 0 1>}
  }
  MaterialExpressionNamedRerouteDeclaration_20 {
   class=MaterialExpressionNamedRerouteDeclaration
   inputs [None {from=MaterialExpressionDynamicParameter_0, output=A}]
   properties {name=SwirlParticleNormAge, nodeColor=<rgba 1 0.7352 0 1>}
  }
  MaterialExpressionNamedRerouteDeclaration_22 {
   class=MaterialExpressionNamedRerouteDeclaration
   inputs [None {from=MaterialExpressionConstant_21}]
   properties {name=SwirlSlowdown, nodeColor=<rgba 1 0.8771 0 1>}
  }
  MaterialExpressionNamedRerouteDeclaration_23 {
   class=MaterialExpressionNamedRerouteDeclaration
   inputs [None {from=MaterialExpressionDivide_5}]
   properties {name=SwirlRotationAmt, nodeColor=<rgba 1 0 0.8458 1>}
  }
  MaterialExpressionNamedRerouteDeclaration_3 {
   class=MaterialExpressionNamedRerouteDeclaration
   inputs [None {from=MaterialExpressionCustom_1, output=SwirlAxis_out}]
   properties {name=SwirlAxis, nodeColor=<rgba 1 0.8095 0 1>}
  }
  MaterialExpressionNamedRerouteDeclaration_4 {
   class=MaterialExpressionNamedRerouteDeclaration
   inputs [None {from=MaterialExpressionSubtract_2}]
   properties {name=SwirlVtxLocalPos, nodeColor=<rgba 1 0.8057 0 1>}
  }
  MaterialExpressionNamedRerouteDeclaration_5 {
   class=MaterialExpressionNamedRerouteDeclaration
   inputs [None {from=MaterialExpressionDynamicParameter_1, output=RGB}]
   properties {name=SwirlOrigPos, nodeColor=<rgba 1 0.8777 0 1>}
  }
  MaterialExpressionNamedRerouteDeclaration_6 {
   class=MaterialExpressionNamedRerouteDeclaration
   inputs [None {from=MaterialExpressionSubtract_3}]
   properties {name=SwirlOffset_FlattenBase, nodeColor=<rgba 1 0 0.7881 1>}
  }
  MaterialExpressionNamedRerouteDeclaration_7 {
   class=MaterialExpressionNamedRerouteDeclaration
   inputs [None {from=MaterialExpressionMultiply_14}]
   properties {name=SwirlOffset_FlattenTip, nodeColor=<rgba 1 0 0.7881 1>}
  }
  MaterialExpressionNamedRerouteDeclaration_8 {
   class=MaterialExpressionNamedRerouteDeclaration
   inputs [None {from=MaterialExpressionIf_1}]
   properties {name=SwirlOffset_Rotate, nodeColor=<rgba 1 0 0.7881 1>}
  }
  MaterialExpressionNamedRerouteDeclaration_9 {
   class=MaterialExpressionNamedRerouteDeclaration
   inputs [None {from=MaterialExpressionMaterialFunctionCall_2, output=WorldPositionOffset}]
   properties {name=SwirlOffset_SplineThicken, nodeColor=<rgba 1 0 0.5156 1>}
  }
  MaterialExpressionNamedRerouteUsage_0 {class=MaterialExpressionNamedRerouteUsage}
  MaterialExpressionNamedRerouteUsage_1 {class=MaterialExpressionNamedRerouteUsage}
  MaterialExpressionNamedRerouteUsage_10 {class=MaterialExpressionNamedRerouteUsage}
  MaterialExpressionNamedRerouteUsage_11 {class=MaterialExpressionNamedRerouteUsage}
  MaterialExpressionNamedRerouteUsage_12 {class=MaterialExpressionNamedRerouteUsage}
  MaterialExpressionNamedRerouteUsage_13 {class=MaterialExpressionNamedRerouteUsage}
  MaterialExpressionNamedRerouteUsage_14 {class=MaterialExpressionNamedRerouteUsage}
  MaterialExpressionNamedRerouteUsage_15 {class=MaterialExpressionNamedRerouteUsage}
  MaterialExpressionNamedRerouteUsage_16 {class=MaterialExpressionNamedRerouteUsage}
  MaterialExpressionNamedRerouteUsage_17 {class=MaterialExpressionNamedRerouteUsage}
  MaterialExpressionNamedRerouteUsage_18 {class=MaterialExpressionNamedRerouteUsage}
  MaterialExpressionNamedRerouteUsage_19 {class=MaterialExpressionNamedRerouteUsage}
  MaterialExpressionNamedRerouteUsage_2 {class=MaterialExpressionNamedRerouteUsage}
  MaterialExpressionNamedRerouteUsage_20 {class=MaterialExpressionNamedRerouteUsage}
  MaterialExpressionNamedRerouteUsage_21 {class=MaterialExpressionNamedRerouteUsage}
  MaterialExpressionNamedRerouteUsage_22 {class=MaterialExpressionNamedRerouteUsage}
  MaterialExpressionNamedRerouteUsage_23 {class=MaterialExpressionNamedRerouteUsage}
  MaterialExpressionNamedRerouteUsage_24 {class=MaterialExpressionNamedRerouteUsage}
  MaterialExpressionNamedRerouteUsage_25 {class=MaterialExpressionNamedRerouteUsage}
  MaterialExpressionNamedRerouteUsage_27 {class=MaterialExpressionNamedRerouteUsage}
  MaterialExpressionNamedRerouteUsage_28 {class=MaterialExpressionNamedRerouteUsage}
  MaterialExpressionNamedRerouteUsage_29 {class=MaterialExpressionNamedRerouteUsage}
  MaterialExpressionNamedRerouteUsage_3 {class=MaterialExpressionNamedRerouteUsage}
  MaterialExpressionNamedRerouteUsage_30 {class=MaterialExpressionNamedRerouteUsage}
  MaterialExpressionNamedRerouteUsage_31 {class=MaterialExpressionNamedRerouteUsage}
  MaterialExpressionNamedRerouteUsage_37 {class=MaterialExpressionNamedRerouteUsage}
  MaterialExpressionNamedRerouteUsage_38 {class=MaterialExpressionNamedRerouteUsage}
  MaterialExpressionNamedRerouteUsage_4 {class=MaterialExpressionNamedRerouteUsage}
  MaterialExpressionNamedRerouteUsage_41 {class=MaterialExpressionNamedRerouteUsage}
  MaterialExpressionNamedRerouteUsage_43 {class=MaterialExpressionNamedRerouteUsage}
  MaterialExpressionNamedRerouteUsage_44 {class=MaterialExpressionNamedRerouteUsage}
  MaterialExpressionNamedRerouteUsage_45 {class=MaterialExpressionNamedRerouteUsage}
  MaterialExpressionNamedRerouteUsage_47 {class=MaterialExpressionNamedRerouteUsage}
  MaterialExpressionNamedRerouteUsage_5 {class=MaterialExpressionNamedRerouteUsage}
  MaterialExpressionNamedRerouteUsage_50 {class=MaterialExpressionNamedRerouteUsage}
  MaterialExpressionNamedRerouteUsage_6 {class=MaterialExpressionNamedRerouteUsage}
  MaterialExpressionNamedRerouteUsage_7 {class=MaterialExpressionNamedRerouteUsage}
  MaterialExpressionNamedRerouteUsage_8 {class=MaterialExpressionNamedRerouteUsage}
  MaterialExpressionNamedRerouteUsage_9 {class=MaterialExpressionNamedRerouteUsage}
  MaterialExpressionNormalize_0 {class=MaterialExpressionNormalize, inputs [VectorInput {from=MaterialExpressionNamedRerouteUsage_8, output=""}]}
  MaterialExpressionNormalize_1 {
   class=MaterialExpressionNormalize
   inputs [VectorInput {from=MaterialExpressionNamedRerouteUsage_28, output=""}]
   properties {desc=sine Wave in direction of swirl}
  }
  MaterialExpressionPanner_0 {
   class=MaterialExpressionPanner
   inputs [
    Coordinate {from=MaterialExpressionMultiply_7}
    Time {from=MaterialExpressionTime_0}
    Speed {from=""}
   ]
   properties {speedX=-1}
  }
  MaterialExpressionPanner_1 {
   class=MaterialExpressionPanner
   inputs [
    Coordinate {from=MaterialExpressionMultiply_25}
    Time {from=MaterialExpressionTime_1}
    Speed {from=MaterialExpressionMultiply_31}
   ]
   properties {speedX=-1}
  }
  MaterialExpressionPanner_2 {
   class=MaterialExpressionPanner
   inputs [
    Coordinate {from=MaterialExpressionTextureCoordinate_3}
    Time {from=MaterialExpressionTime_2}
    Speed {from=""}
   ]
  }
  MaterialExpressionParticlePositionWS_0 {class=MaterialExpressionParticlePositionWS, inputs [], properties {}}
  MaterialExpressionParticlePositionWS_1 {class=MaterialExpressionParticlePositionWS, inputs [], properties {}}
  MaterialExpressionParticleRandom_0 {class=MaterialExpressionParticleRandom, inputs [], properties {}}
  MaterialExpressionParticleRandom_1 {class=MaterialExpressionParticleRandom, inputs [], properties {}}
  MaterialExpressionParticleRandom_10 {class=MaterialExpressionParticleRandom, inputs [], properties {}}
  MaterialExpressionParticleRandom_11 {class=MaterialExpressionParticleRandom, inputs [], properties {}}
  MaterialExpressionParticleRandom_2 {class=MaterialExpressionParticleRandom, inputs [], properties {}}
  MaterialExpressionParticleRandom_3 {class=MaterialExpressionParticleRandom, inputs [], properties {}}
  MaterialExpressionParticleRandom_4 {class=MaterialExpressionParticleRandom, inputs [], properties {}}
  MaterialExpressionParticleRandom_5 {class=MaterialExpressionParticleRandom, inputs [], properties {}}
  MaterialExpressionParticleRandom_6 {class=MaterialExpressionParticleRandom, inputs [], properties {}}
  MaterialExpressionParticleRandom_7 {class=MaterialExpressionParticleRandom, inputs [], properties {}}
  MaterialExpressionParticleRandom_8 {class=MaterialExpressionParticleRandom, inputs [], properties {}}
  MaterialExpressionParticleRandom_9 {class=MaterialExpressionParticleRandom, inputs [], properties {}}
  MaterialExpressionPower_0 {
   class=MaterialExpressionPower
   inputs [
    Base {from=MaterialExpressionAdd_10}
    Exp {from=""}
   ]
  }
  MaterialExpressionPower_1 {
   class=MaterialExpressionPower
   inputs [
    Base {from=MaterialExpressionClamp_5}
    Exp {from=""}
   ]
   properties {constExponent=3}
  }
  MaterialExpressionReroute_0 {class=MaterialExpressionReroute, inputs [None {from=MaterialExpressionDivide_5}]}
  MaterialExpressionReroute_1 {class=MaterialExpressionReroute, inputs [None {from=MaterialExpressionAdd_5}]}
  MaterialExpressionReroute_10 {class=MaterialExpressionReroute, inputs [None {from=MaterialExpressionMaterialFunctionCall_7, output=WPO}]}
  MaterialExpressionReroute_11 {class=MaterialExpressionReroute, inputs [None {from=MaterialExpressionMaterialFunctionCall_26, output=Result}]}
  MaterialExpressionReroute_2 {class=MaterialExpressionReroute, inputs [None {from=MaterialExpressionCustom_0}]}
  MaterialExpressionReroute_3 {class=MaterialExpressionReroute, inputs [None {from=MaterialExpressionSubtract_6}]}
  MaterialExpressionReroute_4 {class=MaterialExpressionReroute, inputs [None {from=MaterialExpressionNamedRerouteDeclaration_14}]}
  MaterialExpressionReroute_5 {class=MaterialExpressionReroute, inputs [None {from=MaterialExpressionCustom_4}]}
  MaterialExpressionReroute_6 {class=MaterialExpressionReroute, inputs [None {from=MaterialExpressionReroute_10}]}
  MaterialExpressionReroute_7 {class=MaterialExpressionReroute, inputs [None {from=MaterialExpressionConstant_47}]}
  MaterialExpressionReroute_8 {class=MaterialExpressionReroute, inputs [None {from=MaterialExpressionReroute_9}]}
  MaterialExpressionReroute_9 {class=MaterialExpressionReroute, inputs [None {from=MaterialExpressionReroute_1}]}
  MaterialExpressionRound_0 {class=MaterialExpressionRound, inputs [None {from=MaterialExpressionParticleRandom_0, output=""}], properties {}}
  MaterialExpressionRound_1 {class=MaterialExpressionRound, inputs [None {from=MaterialExpressionMaterialFunctionCall_27, output=Result}], properties {}}
  MaterialExpressionScalarParameter_0 {class=MaterialExpressionScalarParameter, properties {defaultValue=1, parameterName=GlobalScale}}
  MaterialExpressionScalarParameter_1 {class=MaterialExpressionScalarParameter, properties {defaultValue=0.5, parameterName=FreqVariation}}
  MaterialExpressionScalarParameter_2 {class=MaterialExpressionScalarParameter, properties {defaultValue=0.005, parameterName=SineFreq}}
  MaterialExpressionScalarParameter_3 {class=MaterialExpressionScalarParameter, properties {defaultValue=0.5, parameterName=SineWaveSpeed}}
  MaterialExpressionScalarParameter_4 {class=MaterialExpressionScalarParameter, properties {defaultValue=1, parameterName=DualSineWaveSwitch}}
  MaterialExpressionScalarParameter_5 {class=MaterialExpressionScalarParameter, properties {defaultValue=1.1, parameterName=Width Main}}
  MaterialExpressionScalarParameter_6 {class=MaterialExpressionScalarParameter, properties {defaultValue=8, desc=SineWaveAmp, parameterName=SineAmp}}
  MaterialExpressionSign_0 {class=MaterialExpressionSign, inputs [None {from=MaterialExpressionDotProduct_0, output=""}], properties {}}
  MaterialExpressionSine_0 {class=MaterialExpressionSine, inputs [None {from=MaterialExpressionPanner_0}]}
  MaterialExpressionSine_1 {class=MaterialExpressionSine, inputs [None {from=MaterialExpressionConvert_2}]}
  MaterialExpressionSine_2 {class=MaterialExpressionSine, inputs [None {from=MaterialExpressionMultiply_37}], properties {desc=sinewave}}
  MaterialExpressionStaticBool_0 {class=MaterialExpressionStaticBool}
  MaterialExpressionStaticBool_1 {class=MaterialExpressionStaticBool, properties {value=true}}
  MaterialExpressionStaticSwitchParameter_2 {
   class=MaterialExpressionStaticSwitchParameter
   inputs [
    True {from=MaterialExpressionConstant_19, output=""}
    False {from=MaterialExpressionConstant_18, output=""}
   ]
   properties {defaultValue=true, parameterName=FaceCamera}
  }
  MaterialExpressionSubtract_0 {
   class=MaterialExpressionSubtract
   inputs [
    A {from=MaterialExpressionNamedRerouteUsage_14}
    B {from=MaterialExpressionConstant_1}
   ]
  }
  MaterialExpressionSubtract_1 {
   class=MaterialExpressionSubtract
   inputs [
    A {from=MaterialExpressionMultiply_24}
    B {from=""}
   ]
  }
  MaterialExpressionSubtract_10 {
   class=MaterialExpressionSubtract
   inputs [
    A {from=MaterialExpressionNamedRerouteUsage_24}
    B {from=MaterialExpressionNamedRerouteUsage_25}
   ]
  }
  MaterialExpressionSubtract_11 {
   class=MaterialExpressionSubtract
   inputs [
    A {from=MaterialExpressionParticlePositionWS_0}
    B {from=MaterialExpressionCameraPositionWS_0}
   ]
  }
  MaterialExpressionSubtract_2 {
   class=MaterialExpressionSubtract
   inputs [
    A {from=MaterialExpressionAdd_8}
    B {from=MaterialExpressionNamedRerouteUsage_0}
   ]
  }
  MaterialExpressionSubtract_3 {
   class=MaterialExpressionSubtract
   inputs [
    A {from=MaterialExpressionMultiply_0}
    B {from=MaterialExpressionNamedRerouteUsage_4}
   ]
  }
  MaterialExpressionSubtract_4 {
   class=MaterialExpressionSubtract
   inputs [
    A {from=MaterialExpressionAdd_17}
    B {from=MaterialExpressionNamedRerouteUsage_12}
   ]
  }
  MaterialExpressionSubtract_5 {
   class=MaterialExpressionSubtract
   inputs [
    A {from=MaterialExpressionNamedRerouteUsage_37}
    B {from=MaterialExpressionNamedRerouteUsage_0}
   ]
  }
  MaterialExpressionSubtract_6 {
   class=MaterialExpressionSubtract
   inputs [
    A {from=MaterialExpressionDotProduct_1}
    B {from=MaterialExpressionMin_0}
   ]
   properties {desc=dist past swirlDist}
  }
  MaterialExpressionSubtract_7 {
   class=MaterialExpressionSubtract
   inputs [
    A {from=MaterialExpressionMultiply_3}
    B {from=""}
   ]
  }
  MaterialExpressionSubtract_8 {
   class=MaterialExpressionSubtract
   inputs [
    A {from=MaterialExpressionWorldPosition_2, output=XYZ}
    B {from=MaterialExpressionParticlePositionWS_1}
   ]
  }
  MaterialExpressionSubtract_9 {
   class=MaterialExpressionSubtract
   inputs [
    A {from=""}
    B {from=MaterialExpressionConvert_0}
   ]
  }
  MaterialExpressionTextureCoordinate_0 {class=MaterialExpressionTextureCoordinate}
  MaterialExpressionTextureCoordinate_1 {class=MaterialExpressionTextureCoordinate}
  MaterialExpressionTextureCoordinate_2 {class=MaterialExpressionTextureCoordinate}
  MaterialExpressionTextureCoordinate_3 {class=MaterialExpressionTextureCoordinate, properties {uTiling=0.15, vTiling=0.2}}
  MaterialExpressionTextureCoordinate_4 {class=MaterialExpressionTextureCoordinate}
  MaterialExpressionTextureObject_1 {
   class=MaterialExpressionTextureObject
   inputs []
   properties {samplerType=SAMPLERTYPE_LinearColor, texture=/Game/Global/Textures/Default/Masks_and_Gradients/T_VertGradient}
  }
  MaterialExpressionTextureSample_1 {
   class=MaterialExpressionTextureSample
   inputs [
    UVs {from=MaterialExpressionPanner_2}
    Tex {from=""}
    Apply View MipBias {from=""}
   ]
   properties {samplerType=SAMPLERTYPE_LinearColor, texture=/Game/Athena/Apollo/Water/Textures/02_Rounded/03_WaterFoam_Veroni_base_H}
  }
  MaterialExpressionTextureSample_2 {
   class=MaterialExpressionTextureSample
   inputs [
    UVs {from=MaterialExpressionMultiply_9}
    Tex {from=""}
    Apply View MipBias {from=""}
   ]
   properties {texture=/BRCosmetics/Athena/Effects/Particles/Skydiving/NS_Contrail_SuperSwirl/Textures/T_Contrail_SuperSwirl_Cloud}
  }
  MaterialExpressionTime_0 {class=MaterialExpressionTime}
  MaterialExpressionTime_1 {class=MaterialExpressionTime}
  MaterialExpressionTime_2 {class=MaterialExpressionTime}
  MaterialExpressionWorldPosition_2 {class=MaterialExpressionWorldPosition, inputs [], properties {}}
 }
 properties {
  bAutomaticallySetUsageInEditor=false
  bUsedWithNiagaraMeshParticles=true
  blendMode=BLEND_Translucent
  translucencyLightingMode=TLM_VolumetricPerVertexDirectional
  usedShadingModels=Default Lit
 }
}
