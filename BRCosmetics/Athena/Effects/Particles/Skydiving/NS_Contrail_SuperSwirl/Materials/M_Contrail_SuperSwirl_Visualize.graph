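# Debug-style visualization material for the SuperSwirl skydiving contrail, flagged for use on Niagara mesh particles.
# A SplineThicken function with equal constant base and tip widths produces a world position offset, which is summed with a dynamic parameter channel that particles can drive.
# A separate sample of the centerpoint debug texture uses mesh UVs.
# The graph shows no links from either chain to material outputs.
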Material /BRCosmetics/Athena/Effects/Particles/Skydiving/NS_Contrail_SuperSwirl/Materials/M_Contrail_SuperSwirl_Visualize
graph {
 properties {bAutomaticallySetUsageInEditor=false, bUsedWithNiagaraMeshParticles=true, usedShadingModels=Default Lit}
 nodes [
  n1 = Constant() {r=5}
  Add(
   A=MaterialFunctionCall(WidthBase=@n1, WidthTip=@n1, Expand U or V UV Channel=StaticBool()) {materialFunction=&SplineThicken}.WorldPositionOffset
   B=DynamicParameter() {defaultValue=<rgba 100 100 100 1>}.Param4
  )
  TextureSample(UVs=TextureCoordinate()) {samplerType=SAMPLERTYPE_LinearColor, texture=/Game/Global/Textures/Default/Debugs/CenterpointDebugTexture}
 ]
}
