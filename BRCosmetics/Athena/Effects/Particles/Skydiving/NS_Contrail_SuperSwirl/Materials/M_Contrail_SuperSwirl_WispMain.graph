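# Translucent Niagara mesh-particle material for the wisps of the SuperSwirl skydiving contrail cosmetic.
# SplineThicken sets a sine-pulsed width, while world position offset adds sine and cosine swirls around, and waves along, the axis from an origin dynamic parameter to the particle.
# A cloud texture slides along the strand over particle life, tinted by custom-code gradients with per-particle hue shift.
# Opacity combines cloud alpha, foam-texture erosion at the strand ends and a lifetime ramp.
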Material /BRCosmetics/Athena/Effects/Particles/Skydiving/NS_Contrail_SuperSwirl/Materials/M_Contrail_SuperSwirl_WispMain
graph {
 expressions {
  MaterialExpressionAdd_0 {
   class=MaterialExpressionAdd
   inputs [
    A {from=MaterialExpressionMultiply_25}
    B {from=MaterialExpressionParticleRandom_11}
   ]
  }
  MaterialExpressionAdd_1 {
   class=MaterialExpressionAdd
   inputs [
    A {from=MaterialExpressionMultiply_20}
    B {from=MaterialExpressionMultiply_8}
   ]
  }
  MaterialExpressionAdd_12 {
   class=MaterialExpressionAdd
   inputs [
    A {from=MaterialExpressionMaterialFunctionCall_2, output=WorldPositionOffset}
    B {from=MaterialExpressionAdd_16}
   ]
  }
  MaterialExpressionAdd_15 {
   class=MaterialExpressionAdd
   inputs [
    A {from=MaterialExpressionAdd_2}
    B {from=""}
   ]
   properties {constB=0.01}
  }
  MaterialExpressionAdd_16 {
   class=MaterialExpressionAdd
   inputs [
    A {from=MaterialExpressionMultiply_6}
    B {from=MaterialExpressionMultiply_41}
   ]
   properties {constB=0}
  }
  MaterialExpressionAdd_2 {
   class=MaterialExpressionAdd
   inputs [
    A {from=MaterialExpressionSubtract_0}
    B {from=MaterialExpressionSmoothStep_1}
   ]
  }
  MaterialExpressionAdd_4 {
   class=MaterialExpressionAdd
   inputs [
    A {from=MaterialExpressionCustom_4}
    B {from=MaterialExpressionMultiply_32}
   ]
  }
  MaterialExpressionAdd_5 {
   class=MaterialExpressionAdd
   inputs [
    A {from=MaterialExpressionMaterialFunctionCall_5, output=Result}
    B {from=MaterialExpressionMultiply_0}
   ]
  }
  MaterialExpressionAdd_8 {
   class=MaterialExpressionAdd
   inputs [
    A {from=MaterialExpressionMultiply_18}
    B {from=MaterialExpressionParticleRandom_7}
   ]
  }
  MaterialExpressionAdd_9 {
   class=MaterialExpressionAdd
   inputs [
    A {from=MaterialExpressionSine_0}
    B {from=MaterialExpressionScalarParameter_22}
   ]
   properties {constB=0.5}
  }
  MaterialExpressionClamp_0 {
   class=MaterialExpressionClamp
   inputs [
    None {from=MaterialExpressionAdd_4, output=""}
    Min {from="", output=""}
    Max {from="", output=""}
   ]
   properties {}
  }
  MaterialExpressionClamp_3 {
   class=MaterialExpressionClamp
   inputs [
    None {from=MaterialExpressionTextureSample_2, output=RGB}
    Min {from="", output=""}
    Max {from="", output=""}
   ]
   properties {desc=lift blacks, minDefault=0.2}
  }
  MaterialExpressionClamp_4 {
   class=MaterialExpressionClamp
   inputs [
    None {from=MaterialExpressionMaterialFunctionCall_14, output=Result}
    Min {from="", output=""}
    Max {from="", output=""}
   ]
   properties {}
  }
  MaterialExpressionConstant_11 {class=MaterialExpressionConstant, properties {r=0.4}}
  MaterialExpressionConstant_14 {class=MaterialExpressionConstant, properties {r=1}}
  MaterialExpressionConstant_17 {class=MaterialExpressionConstant, properties {r=3}}
  MaterialExpressionConstant_24 {class=MaterialExpressionConstant, properties {r=7}}
  MaterialExpressionConstant_27 {class=MaterialExpressionConstant, properties {r=7}}
  MaterialExpressionConstant_45 {class=MaterialExpressionConstant, properties {r=22.2}}
  MaterialExpressionConstant_64 {class=MaterialExpressionConstant, properties {r=44}}
  MaterialExpressionConvert_0 {
   class=MaterialExpressionConvert
   inputs [None {from=MaterialExpressionReroute_2}]
   properties {
    convertInputs [{
      type=Vector2
      expressionInput {
       expression=&self:MaterialExpressionReroute_2
       inputName=None
       mask=0
       maskA=0
       maskB=0
       maskG=0
       maskR=0
       outputIndex=0
      }
     }]
    convertOutputs [{type=Scalar}, {type=Scalar}]
   }
  }
  MaterialExpressionConvert_1 {
   class=MaterialExpressionConvert
   inputs [
    None {from=MaterialExpressionMultiply_17}
    None {from=MaterialExpressionTextureSample_2, output=A}
   ]
   properties {
    convertInputs [
     {
      type=Vector3
      expressionInput {
       expression=&self:MaterialExpressionMultiply_17
       inputName=None
       mask=0
       maskA=0
       maskB=0
       maskG=0
       maskR=0
       outputIndex=0
      }
     }
     {
      type=Scalar
      expressionInput {
       expression=&self:MaterialExpressionTextureSample_2
       inputName=None
       mask=1
       maskA=1
       maskB=0
       maskG=0
       maskR=0
       outputIndex=4
      }
     }
    ]
    convertOutputs [{type=Vector4, defaultValue=<rgba 0 0 0 1>}]
   }
  }
  MaterialExpressionConvert_2 {
   class=MaterialExpressionConvert
   inputs [None {from=MaterialExpressionPanner_1}]
   properties {
    convertInputs [{
      type=Vector2
      expressionInput {
       expression=&self:MaterialExpressionPanner_1
       inputName=None
       mask=0
       maskA=0
       maskB=0
       maskG=0
       maskR=0
       outputIndex=0
      }
     }]
    convertOutputs [{type=Scalar}]
   }
  }
  MaterialExpressionConvert_3 {
   class=MaterialExpressionConvert
   inputs [
    None {from=MaterialExpressionAdd_5}
    None {from=MaterialExpressionMaterialFunctionCall_3, output=G}
   ]
   properties {
    convertInputs [
     {
      type=Scalar
      expressionInput {
       expression=&self:MaterialExpressionAdd_5
       inputName=None
       mask=0
       maskA=0
       maskB=0
       maskG=0
       maskR=0
       outputIndex=0
      }
     }
     {
      type=Scalar
      expressionInput {
       expression=&self:MaterialExpressionMaterialFunctionCall_3
       inputName=None
       mask=0
       maskA=0
       maskB=0
       maskG=0
       maskR=0
       outputIndex=1
      }
     }
    ]
    convertOutputs [{type=Vector2}]
   }
  }
  MaterialExpressionConvert_4 {
   class=MaterialExpressionConvert
   inputs [None {from=MaterialExpressionNamedRerouteUsage_2}]
   properties {
    convertInputs [{
      type=Vector2
      expressionInput {
       expression=&self:MaterialExpressionNamedRerouteUsage_2
       inputName=None
       mask=0
       maskA=0
       maskB=0
       maskG=0
       maskR=0
       outputIndex=0
      }
     }]
    convertOutputs [{type=Scalar}]
   }
  }
  MaterialExpressionConvert_5 {
   class=MaterialExpressionConvert
   inputs [None {from=MaterialExpressionNamedRerouteUsage_7}]
   properties {
    convertInputs [{
      type=Vector2
      expressionInput {
       expression=&self:MaterialExpressionNamedRerouteUsage_7
       inputName=None
       mask=0
       maskA=0
       maskB=0
       maskG=0
       maskR=0
       outputIndex=0
      }
     }]
    convertOutputs [{type=Scalar}]
   }
  }
  MaterialExpressionConvert_6 {
   class=MaterialExpressionConvert
   inputs [None {from=MaterialExpressionMultiply_50}]
   properties {
    convertInputs [{
      type=Scalar
      expressionInput {
       expression=&self:MaterialExpressionMultiply_50
       inputName=None
       mask=0
       maskA=0
       maskB=0
       maskG=0
       maskR=0
       outputIndex=0
      }
     }]
    convertOutputs [{type=Vector3}]
   }
  }
  MaterialExpressionConvert_7 {
   class=MaterialExpressionConvert
   inputs [None {from=MaterialExpressionPanner_6}]
   properties {
    convertInputs [{
      type=Vector2
      expressionInput {
       expression=&self:MaterialExpressionPanner_6
       inputName=None
       mask=0
       maskA=0
       maskB=0
       maskG=0
       maskR=0
       outputIndex=0
      }
     }]
    convertOutputs [{type=Scalar}]
   }
  }
  MaterialExpressionCosine_0 {class=MaterialExpressionCosine, inputs [None {from=MaterialExpressionConvert_2, output=""}], properties {}}
  MaterialExpressionCrossProduct_0 {
   class=MaterialExpressionCrossProduct
   inputs [
    A {from=MaterialExpressionMaterialFunctionCall_36, output=Result}
    B {from=MaterialExpressionSubtract_2, output=""}
   ]
   properties {}
  }
  MaterialExpressionCrossProduct_1 {
   class=MaterialExpressionCrossProduct
   inputs [
    A {from=MaterialExpressionCrossProduct_0, output=""}
    B {from=MaterialExpressionSubtract_2, output=""}
   ]
   properties {}
  }
  MaterialExpressionCustom_0 {
   class=MaterialExpressionCustom
   inputs [
    freq {from=MaterialExpressionScalarParameter_8, output=""}
    freq_var {from=MaterialExpressionScalarParameter_9, output=""}
   ]
   properties {
    code="min_out = (1-freq_var)*freq;\r\nmax_out = (1+freq_var)*freq;\r\nreturn 1;"
    showCode=true
    additionalOutputs [{outputName=min_out, outputType=CMOT_Float1}, {outputName=max_out, outputType=CMOT_Float1}]
    inputs [
     {
      inputName=freq
      input {
       expression=&self:MaterialExpressionScalarParameter_8
       inputName=None
       mask=0
       maskA=0
       maskB=0
       maskG=0
       maskR=0
       outputIndex=0
      }
     }
     {
      inputName=freq_var
      input {
       expression=&self:MaterialExpressionScalarParameter_9
       inputName=None
       mask=0
       maskA=0
       maskB=0
       maskG=0
       maskR=0
       outputIndex=0
      }
     }
    ]
   }
  }
  MaterialExpressionCustom_1 {
   class=MaterialExpressionCustom
   inputs [
    val {from=MaterialExpressionScalarParameter_10, output=""}
    val_var {from=MaterialExpressionScalarParameter_11, output=""}
   ]
   properties {
    code="min_out = -1*(1-val_var)*val;\r\nmax_out = -1*(1+val_var)*val;\r\nreturn 1;"
    showCode=true
    additionalOutputs [{outputName=min_out, outputType=CMOT_Float1}, {outputName=max_out, outputType=CMOT_Float1}]
    inputs [
     {
      inputName=val
      input {
       expression=&self:MaterialExpressionScalarParameter_10
       inputName=None
       mask=0
       maskA=0
       maskB=0
       maskG=0
       maskR=0
       outputIndex=0
      }
     }
     {
      inputName=val_var
      input {
       expression=&self:MaterialExpressionScalarParameter_11
       inputName=None
       mask=0
       maskA=0
       maskB=0
       maskG=0
       maskR=0
       outputIndex=0
      }
     }
    ]
   }
  }
  MaterialExpressionCustom_2 {
   class=MaterialExpressionCustom
   inputs [
    freq {from=MaterialExpressionScalarParameter_18, output=""}
    freq_var {from=MaterialExpressionScalarParameter_19, output=""}
   ]
   properties {
    code="min_out = (1-freq_var)*freq;\r\nmax_out = (1+freq_var)*freq;\r\nreturn 1;"
    showCode=true
    additionalOutputs [{outputName=min_out, outputType=CMOT_Float1}, {outputName=max_out, outputType=CMOT_Float1}]
    inputs [
     {
      inputName=freq
      input {
       expression=&self:MaterialExpressionScalarParameter_18
       inputName=None
       mask=0
       maskA=0
       maskB=0
       maskG=0
       maskR=0
       outputIndex=0
      }
     }
     {
      inputName=freq_var
      input {
       expression=&self:MaterialExpressionScalarParameter_19
       inputName=None
       mask=0
       maskA=0
       maskB=0
       maskG=0
       maskR=0
       outputIndex=0
      }
     }
    ]
   }
  }
  MaterialExpressionCustom_3 {
   class=MaterialExpressionCustom
   inputs [
    val {from=MaterialExpressionScalarParameter_20, output=""}
    val_var {from=MaterialExpressionScalarParameter_21, output=""}
   ]
   properties {
    code="min_out = (1-val_var)*val;\r\nmax_out = (1+val_var)*val;\r\nreturn 1;"
    showCode=true
    additionalOutputs [{outputName=min_out, outputType=CMOT_Float1}, {outputName=max_out, outputType=CMOT_Float1}]
    inputs [
     {
      inputName=val
      input {
       expression=&self:MaterialExpressionScalarParameter_20
       inputName=None
       mask=0
       maskA=0
       maskB=0
       maskG=0
       maskR=0
       outputIndex=0
      }
     }
     {
      inputName=val_var
      input {
       expression=&self:MaterialExpressionScalarParameter_21
       inputName=None
       mask=0
       maskA=0
       maskB=0
       maskG=0
       maskR=0
       outputIndex=0
      }
     }
    ]
   }
  }
  MaterialExpressionCustom_4 {
   class=MaterialExpressionCustom
   inputs [Alpha {from=MaterialExpressionConvert_0, output=""}]
   properties {
    code="float4 c1 = {0,0,0,1};\r\nfloat4 c2 = {.1, .5, .9,1};\r\nreturn lerp(c1, c2, Alpha); "
    outputType=CMOT_Float4
    inputs [{
      inputName=Alpha
      input {
       expression=&self:MaterialExpressionConvert_0
       inputName=None
       mask=0
       maskA=0
       maskB=0
       maskG=0
       maskR=0
       outputIndex=0
      }
     }]
   }
  }
  MaterialExpressionCustom_5 {
   class=MaterialExpressionCustom
   inputs [
    val {from=MaterialExpressionScalarParameter_24, output=""}
    val_var {from=MaterialExpressionScalarParameter_25, output=""}
   ]
   properties {
    code="min_out = -1*(1-val_var)*val;\r\nmax_out = -1*(1+val_var)*val;\r\nreturn 1;"
    showCode=true
    additionalOutputs [{outputName=min_out, outputType=CMOT_Float1}, {outputName=max_out, outputType=CMOT_Float1}]
    inputs [
     {
      inputName=val
      input {
       expression=&self:MaterialExpressionScalarParameter_24
       inputName=None
       mask=0
       maskA=0
       maskB=0
       maskG=0
       maskR=0
       outputIndex=0
      }
     }
     {
      inputName=val_var
      input {
       expression=&self:MaterialExpressionScalarParameter_25
       inputName=None
       mask=0
       maskA=0
       maskB=0
       maskG=0
       maskR=0
       outputIndex=0
      }
     }
    ]
   }
  }
  MaterialExpressionCustom_6 {
   class=MaterialExpressionCustom
   inputs [Alpha {from=MaterialExpressionConvert_0, output=""}]
   properties {
    code="float4 c1 = {1,.09,.7,1};\r\nfloat4 c2 = {1, .6, .14,1};\r\nreturn lerp(c1, c2, Alpha); "
    outputType=CMOT_Float4
    inputs [{
      inputName=Alpha
      input {
       expression=&self:MaterialExpressionConvert_0
       inputName=None
       mask=0
       maskA=0
       maskB=0
       maskG=0
       maskR=0
       outputIndex=1
      }
     }]
   }
  }
  MaterialExpressionCustom_7 {
   class=MaterialExpressionCustom
   inputs [Alpha {from=MaterialExpressionParticleRelativeTime_1, output=""}]
   properties {
    code="float rampin = .11;\r\nfloat rampout = .87;\r\nfloat ramp_val = .33;\r\n\r\nfloat a1 = min(1, Alpha * (1/rampin));\r\nfloat a2 = min(1, (1-Alpha)*(1/(1-rampout)));\r\nreturn lerp(ramp_val, 1, a1*a2); "
    outputType=CMOT_Float4
    inputs [{
      inputName=Alpha
      input {
       expression=&self:MaterialExpressionParticleRelativeTime_1
       inputName=None
       mask=0
       maskA=0
       maskB=0
       maskG=0
       maskR=0
       outputIndex=0
      }
     }]
   }
  }
  MaterialExpressionDesaturation_0 {
   class=MaterialExpressionDesaturation
   inputs [
    None {from=MaterialExpressionMaterialFunctionCall_26, output=Result}
    Fraction {from=MaterialExpressionScalarParameter_12, output=""}
   ]
   properties {}
  }
  MaterialExpressionDivide_1 {
   class=MaterialExpressionDivide
   inputs [
    A {from=""}
    B {from=MaterialExpressionMaterialFunctionCall_23, output=Result}
   ]
  }
  MaterialExpressionDynamicParameter_1 {
   class=MaterialExpressionDynamicParameter
   inputs []
   properties {parameterIndex=1, defaultValue=<rgba 1 0.9428 0.0132 1>, paramNames [origposx,origpoxy,origposz,None]}
  }
  MaterialExpressionDynamicParameter_4 {
   class=MaterialExpressionDynamicParameter
   inputs []
   properties {parameterIndex=1, defaultValue=<rgba 1 0.9428 0.0132 1>, paramNames [origposx,origpoxy,origposz,None]}
  }
  MaterialExpressionIf_0 {
   class=MaterialExpressionIf
   inputs [
    A {from=MaterialExpressionScalarParameter_26, output=""}
    B {from="", output=""}
    "A > B" {from=MaterialExpressionConvert_6, output=""}
    "A == B" {from=MaterialExpressionConvert_6, output=""}
    "A < B" {from=MaterialExpressionDesaturation_0, output=""}
   ]
   properties {constB=0.5}
  }
  MaterialExpressionIf_2 {
   class=MaterialExpressionIf
   inputs [
    A {from=MaterialExpressionScalarParameter_13, output=""}
    B {from="", output=""}
    "A > B" {from=MaterialExpressionSubtract_3, output=""}
    "A == B" {from=MaterialExpressionSubtract_3, output=""}
    "A < B" {from=MaterialExpressionTextureCoordinate_2, output=""}
   ]
   properties {constB=0.5}
  }
  MaterialExpressionLinearInterpolate_0 {
   class=MaterialExpressionLinearInterpolate
   inputs [
    A {from=MaterialExpressionScalarParameter_27}
    B {from=MaterialExpressionScalarParameter_28}
    Alpha {from=MaterialExpressionParticleRelativeTime_3}
   ]
   properties {constA=2, constB=0.5}
  }
  MaterialExpressionMaterialFunctionCall_1 {
   class=MaterialExpressionMaterialFunctionCall
   inputs [
    Input {from=MaterialExpressionParticleRandom_1}
    Input Low {from=""}
    Input High {from=""}
    Target Low {from=MaterialExpressionCustom_5, output=min_out}
    Target High {from=MaterialExpressionCustom_5, output=min_out}
   ]
   properties {desc=panning speed, materialFunction=&RemapValueRange}
  }
  MaterialExpressionMaterialFunctionCall_14 {
   class=MaterialExpressionMaterialFunctionCall
   inputs [
    Input {from=MaterialExpressionTextureSample_1, output=R}
    Input Low {from=MaterialExpressionAdd_2}
    Input High {from=MaterialExpressionAdd_15}
    Target Low {from=""}
    Target High {from=""}
   ]
   properties {materialFunction=&RemapValueRange}
  }
  MaterialExpressionMaterialFunctionCall_16 {
   class=MaterialExpressionMaterialFunctionCall
   inputs [
    Input {from=MaterialExpressionParticleRandom_10}
    min {from=MaterialExpressionMultiply_2}
    max {from=MaterialExpressionScalarParameter_7}
    seed {from=MaterialExpressionConstant_45}
   ]
   properties {materialFunction=/InterchangeAssets/Functions/MX_RandomFloat}
  }
  MaterialExpressionMaterialFunctionCall_2 {
   class=MaterialExpressionMaterialFunctionCall
   inputs [
    WidthBase {from=MaterialExpressionReroute_4}
    WidthTip {from=MaterialExpressionReroute_4}
    WorldPosition {from=""}
    UVs for Projection {from=""}
    UVs for Thickness {from=""}
    UVs for texturing {from=""}
    Expand U or V UV Channel {from=MaterialExpressionStaticBool_0}
    DeriveNormalZ {from=MaterialExpressionStaticBool_2}
    AngleCorrectedNormals {from=""}
    AdditionalNormal {from=""}
    FlattenPixelNormal {from=""}
    UVs For NormalShading {from=""}
   ]
   properties {materialFunction=&SplineThicken}
  }
  MaterialExpressionMaterialFunctionCall_23 {
   class=MaterialExpressionMaterialFunctionCall
   inputs [
    Input {from=MaterialExpressionParticleRandom_5}
    min {from=MaterialExpressionScalarParameter_1}
    max {from=MaterialExpressionScalarParameter_2}
    seed {from=MaterialExpressionConstant_24}
   ]
   properties {materialFunction=&MX_RandomFloat}
  }
  MaterialExpressionMaterialFunctionCall_26 {
   class=MaterialExpressionMaterialFunctionCall
   inputs [
    Hue Shift Percentage {from=MaterialExpressionMaterialFunctionCall_16, output=Result}
    Texture {from=MaterialExpressionMultiply_36}
   ]
   properties {materialFunction=&HueShift}
  }
  MaterialExpressionMaterialFunctionCall_27 {
   class=MaterialExpressionMaterialFunctionCall
   inputs [
    Input {from=MaterialExpressionParticleRandom_6}
    min {from=MaterialExpressionCustom_3, output=min_out}
    max {from=MaterialExpressionCustom_3, output=min_out}
    seed {from=MaterialExpressionConstant_27}
   ]
   properties {materialFunction=&MX_RandomFloat}
  }
  MaterialExpressionMaterialFunctionCall_3 {
   class=MaterialExpressionMaterialFunctionCall
   inputs [Float2 {from=MaterialExpressionIf_2}]
   properties {materialFunction=&BreakOutFloat2Components}
  }
  MaterialExpressionMaterialFunctionCall_33 {
   class=MaterialExpressionMaterialFunctionCall
   inputs [Float2 {from=MaterialExpressionTextureCoordinate_0}]
   properties {materialFunction=&BreakOutFloat2Components}
  }
  MaterialExpressionMaterialFunctionCall_36 {
   class=MaterialExpressionMaterialFunctionCall
   inputs []
   properties {materialFunction=&CameraDirectionVector}
  }
  MaterialExpressionMaterialFunctionCall_38 {
   class=MaterialExpressionMaterialFunctionCall
   inputs [
    Input {from=MaterialExpressionParticleRandom_11}
    Input Low {from=""}
    Input High {from=""}
    Target Low {from=MaterialExpressionCustom_1, output=min_out}
    Target High {from=MaterialExpressionCustom_1, output=min_out}
   ]
   properties {desc=panning speed, materialFunction=&RemapValueRange}
  }
  MaterialExpressionMaterialFunctionCall_44 {
   class=MaterialExpressionMaterialFunctionCall
   inputs [
    Input {from=MaterialExpressionParticleRandom_17}
    min {from=MaterialExpressionCustom_0, output=min_out}
    max {from=MaterialExpressionCustom_0, output=min_out}
    seed {from=MaterialExpressionConstant_64}
   ]
   properties {materialFunction=&MX_RandomFloat}
  }
  MaterialExpressionMaterialFunctionCall_5 {
   class=MaterialExpressionMaterialFunctionCall
   inputs [
    Input {from=MaterialExpressionParticleRelativeTime_0}
    Input Low {from=""}
    Input High {from=""}
    Target Low {from=MaterialExpressionMultiply_3}
    Target High {from=MaterialExpressionConstant_14}
   ]
   properties {materialFunction=&RemapValueRange}
  }
  MaterialExpressionMaterialFunctionCall_6 {
   class=MaterialExpressionMaterialFunctionCall
   inputs [
    Input {from=MaterialExpressionMin_0}
    Input Low {from=""}
    Input High {from=MaterialExpressionConstant_11}
    Target Low {from=MaterialExpressionScalarParameter_14}
    Target High {from=MaterialExpressionScalarParameter_15}
   ]
   properties {materialFunction=&RemapValueRange}
  }
  MaterialExpressionMaterialFunctionCall_7 {
   class=MaterialExpressionMaterialFunctionCall
   inputs [
    Input {from=MaterialExpressionMax_0}
    Input Low {from=MaterialExpressionSubtract_4}
    Input High {from=""}
    Target Low {from=MaterialExpressionMaterialFunctionCall_6, output=Result}
    Target High {from=MaterialExpressionScalarParameter_3}
   ]
   properties {materialFunction=&RemapValueRange}
  }
  MaterialExpressionMaterialFunctionCall_8 {
   class=MaterialExpressionMaterialFunctionCall
   inputs [
    Input {from=MaterialExpressionParticleRandom_2}
    min {from=MaterialExpressionCustom_2, output=min_out}
    max {from=MaterialExpressionCustom_2, output=min_out}
    seed {from=MaterialExpressionConstant_17}
   ]
   properties {materialFunction=&MX_RandomFloat}
  }
  MaterialExpressionMax_0 {
   class=MaterialExpressionMax
   inputs [
    A {from=MaterialExpressionParticleRelativeTime_2}
    B {from=MaterialExpressionSubtract_4}
   ]
  }
  MaterialExpressionMin_0 {
   class=MaterialExpressionMin
   inputs [
    A {from=MaterialExpressionParticleRelativeTime_2, output=""}
    B {from=MaterialExpressionConstant_11, output=""}
   ]
   properties {}
  }
  MaterialExpressionMultiply_0 {
   class=MaterialExpressionMultiply
   inputs [
    A {from=MaterialExpressionDivide_1}
    B {from=MaterialExpressionMaterialFunctionCall_3, output=R}
   ]
   properties {constB=2}
  }
  MaterialExpressionMultiply_1 {
   class=MaterialExpressionMultiply
   inputs [
    A {from=MaterialExpressionNamedRerouteUsage_9}
    B {from=MaterialExpressionScalarParameter_16}
   ]
  }
  MaterialExpressionMultiply_10 {
   class=MaterialExpressionMultiply
   inputs [
    A {from=MaterialExpressionIf_0}
    B {from=""}
   ]
   properties {constB=0.8}
  }
  MaterialExpressionMultiply_11 {
   class=MaterialExpressionMultiply
   inputs [
    A {from=MaterialExpressionLinearInterpolate_0}
    B {from=MaterialExpressionMultiply_12}
   ]
  }
  MaterialExpressionMultiply_12 {
   class=MaterialExpressionMultiply
   inputs [
    A {from=MaterialExpressionAdd_9}
    B {from=MaterialExpressionMaterialFunctionCall_27, output=Result}
   ]
   properties {constB=15}
  }
  MaterialExpressionMultiply_16 {
   class=MaterialExpressionMultiply
   inputs [
    A {from=MaterialExpressionMaterialFunctionCall_8, output=Result}
    B {from=MaterialExpressionConvert_4}
   ]
   properties {constA=5, desc=Freq}
  }
  MaterialExpressionMultiply_17 {
   class=MaterialExpressionMultiply
   inputs [
    A {from=MaterialExpressionClamp_3}
    B {from=MaterialExpressionTextureSample_2, output=A}
   ]
  }
  MaterialExpressionMultiply_18 {
   class=MaterialExpressionMultiply
   inputs [
    A {from=MaterialExpressionReroute_3}
    B {from=MaterialExpressionScalarParameter_4}
   ]
   properties {constB=0.3}
  }
  MaterialExpressionMultiply_2 {
   class=MaterialExpressionMultiply
   inputs [
    A {from=""}
    B {from=MaterialExpressionScalarParameter_7}
   ]
   properties {constA=-1}
  }
  MaterialExpressionMultiply_20 {
   class=MaterialExpressionMultiply
   inputs [
    A {from=MaterialExpressionNormalize_0}
    B {from=MaterialExpressionSine_1}
   ]
  }
  MaterialExpressionMultiply_25 {
   class=MaterialExpressionMultiply
   inputs [
    A {from=MaterialExpressionMaterialFunctionCall_44, output=Result}
    B {from=MaterialExpressionConvert_5}
   ]
  }
  MaterialExpressionMultiply_3 {
   class=MaterialExpressionMultiply
   inputs [
    A {from=""}
    B {from=MaterialExpressionDivide_1}
   ]
   properties {constA=-1, constB=-1}
  }
  MaterialExpressionMultiply_32 {
   class=MaterialExpressionMultiply
   inputs [
    A {from=MaterialExpressionSubtract_9}
    B {from=MaterialExpressionCustom_6}
   ]
  }
  MaterialExpressionMultiply_36 {
   class=MaterialExpressionMultiply
   inputs [
    A {from=MaterialExpressionClamp_0}
    B {from=MaterialExpressionConvert_1}
   ]
  }
  MaterialExpressionMultiply_38 {
   class=MaterialExpressionMultiply
   inputs [
    A {from=MaterialExpressionSine_5}
    B {from=MaterialExpressionMultiply_1}
   ]
   properties {constB=100, desc=sin amt}
  }
  MaterialExpressionMultiply_4 {
   class=MaterialExpressionMultiply
   inputs [
    A {from=MaterialExpressionMultiply_50}
    B {from=MaterialExpressionScalarParameter_0}
   ]
  }
  MaterialExpressionMultiply_40 {
   class=MaterialExpressionMultiply
   inputs [
    A {from=MaterialExpressionMultiply_7}
    B {from=MaterialExpressionReroute_0}
   ]
  }
  MaterialExpressionMultiply_41 {
   class=MaterialExpressionMultiply
   inputs [
    A {from=MaterialExpressionNormalize_3}
    B {from=MaterialExpressionMultiply_38}
   ]
  }
  MaterialExpressionMultiply_42 {
   class=MaterialExpressionMultiply
   inputs [
    A {from=MaterialExpressionClamp_4}
    B {from=MaterialExpressionCustom_7}
   ]
  }
  MaterialExpressionMultiply_5 {
   class=MaterialExpressionMultiply
   inputs [
    A {from=MaterialExpressionMaterialFunctionCall_1, output=Result}
    B {from=""}
   ]
   properties {constB=-1}
  }
  MaterialExpressionMultiply_50 {
   class=MaterialExpressionMultiply
   inputs [
    A {from=MaterialExpressionTextureSample_2, output=A}
    B {from=MaterialExpressionMultiply_42}
   ]
  }
  MaterialExpressionMultiply_6 {
   class=MaterialExpressionMultiply
   inputs [
    A {from=MaterialExpressionAdd_1}
    B {from=MaterialExpressionMaterialFunctionCall_7, output=Result}
   ]
   properties {constB=100, desc=SineAmp}
  }
  MaterialExpressionMultiply_7 {
   class=MaterialExpressionMultiply
   inputs [
    A {from=MaterialExpressionNamedRerouteUsage_8}
    B {from=MaterialExpressionScalarParameter_17}
   ]
   properties {constB=0.5}
  }
  MaterialExpressionMultiply_8 {
   class=MaterialExpressionMultiply
   inputs [
    A {from=MaterialExpressionNormalize_1}
    B {from=MaterialExpressionCosine_0}
   ]
  }
  MaterialExpressionMultiply_9 {
   class=MaterialExpressionMultiply
   inputs [
    A {from=MaterialExpressionIf_0}
    B {from=""}
   ]
   properties {constB=0.4}
  }
  MaterialExpressionNamedRerouteDeclaration_0 {
   class=MaterialExpressionNamedRerouteDeclaration
   inputs [None {from=MaterialExpressionConvert_3}]
   properties {name=AnimatedTexCoords, nodeColor=<rgba 1 0.1172 0 1>}
  }
  MaterialExpressionNamedRerouteDeclaration_2 {
   class=MaterialExpressionNamedRerouteDeclaration
   inputs [None {from=MaterialExpressionIf_2}]
   properties {name=TexCoordAdj, nodeColor=<rgba 0.0547 1 0 1>}
  }
  MaterialExpressionNamedRerouteDeclaration_3 {
   class=MaterialExpressionNamedRerouteDeclaration
   inputs [None {from=MaterialExpressionMaterialFunctionCall_44, output=Result}]
   properties {name=SineFreq, nodeColor=<rgba 1 0 0.4687 1>}
  }
  MaterialExpressionNamedRerouteDeclaration_4 {
   class=MaterialExpressionNamedRerouteDeclaration
   inputs [None {from=MaterialExpressionMaterialFunctionCall_7, output=Result}]
   properties {name=SineAmp, nodeColor=<rgba 0 0.7891 1 1>}
  }
  MaterialExpressionNamedRerouteUsage_2 {class=MaterialExpressionNamedRerouteUsage}
  MaterialExpressionNamedRerouteUsage_7 {class=MaterialExpressionNamedRerouteUsage}
  MaterialExpressionNamedRerouteUsage_8 {class=MaterialExpressionNamedRerouteUsage}
  MaterialExpressionNamedRerouteUsage_9 {class=MaterialExpressionNamedRerouteUsage}
  MaterialExpressionNormalize_0 {class=MaterialExpressionNormalize, inputs [VectorInput {from=MaterialExpressionCrossProduct_0, output=""}]}
  MaterialExpressionNormalize_1 {class=MaterialExpressionNormalize, inputs [VectorInput {from=MaterialExpressionCrossProduct_1, output=""}]}
  MaterialExpressionNormalize_3 {class=MaterialExpressionNormalize, inputs [VectorInput {from=MaterialExpressionSubtract_5, output=""}]}
  MaterialExpressionPanner_1 {
   class=MaterialExpressionPanner
   inputs [
    Coordinate {from=MaterialExpressionAdd_0}
    Time {from=MaterialExpressionTime_1}
    Speed {from=MaterialExpressionMaterialFunctionCall_38, output=Result}
   ]
   properties {speedX=-1}
  }
  MaterialExpressionPanner_3 {
   class=MaterialExpressionPanner
   inputs [
    Coordinate {from=MaterialExpressionMultiply_16}
    Time {from=MaterialExpressionTime_3}
    Speed {from=MaterialExpressionScalarParameter_23}
   ]
   properties {speedX=1.5}
  }
  MaterialExpressionPanner_6 {
   class=MaterialExpressionPanner
   inputs [
    Coordinate {from=MaterialExpressionMultiply_40}
    Time {from=MaterialExpressionTime_6}
    Speed {from=MaterialExpressionMultiply_5}
   ]
   properties {speedX=-0.6}
  }
  MaterialExpressionParticlePositionWS_0 {class=MaterialExpressionParticlePositionWS, inputs [], properties {}}
  MaterialExpressionParticlePositionWS_1 {class=MaterialExpressionParticlePositionWS, inputs [], properties {}}
  MaterialExpressionParticleRandom_1 {class=MaterialExpressionParticleRandom, inputs [], properties {}}
  MaterialExpressionParticleRandom_10 {class=MaterialExpressionParticleRandom, inputs [], properties {}}
  MaterialExpressionParticleRandom_11 {class=MaterialExpressionParticleRandom, inputs [], properties {}}
  MaterialExpressionParticleRandom_17 {class=MaterialExpressionParticleRandom, inputs [], properties {}}
  MaterialExpressionParticleRandom_2 {class=MaterialExpressionParticleRandom, inputs [], properties {}}
  MaterialExpressionParticleRandom_5 {class=MaterialExpressionParticleRandom, inputs [], properties {}}
  MaterialExpressionParticleRandom_6 {class=MaterialExpressionParticleRandom, inputs [], properties {}}
  MaterialExpressionParticleRandom_7 {class=MaterialExpressionParticleRandom, inputs [], properties {}}
  MaterialExpressionParticleRelativeTime_0 {class=MaterialExpressionParticleRelativeTime, inputs [], properties {}}
  MaterialExpressionParticleRelativeTime_1 {class=MaterialExpressionParticleRelativeTime, inputs [], properties {}}
  MaterialExpressionParticleRelativeTime_2 {class=MaterialExpressionParticleRelativeTime, inputs [], properties {}}
  MaterialExpressionParticleRelativeTime_3 {class=MaterialExpressionParticleRelativeTime, inputs [], properties {}}
  MaterialExpressionReroute_0 {class=MaterialExpressionReroute, inputs [None {from=MaterialExpressionConvert_5}]}
  MaterialExpressionReroute_1 {class=MaterialExpressionReroute, inputs [None {from=MaterialExpressionMaterialFunctionCall_33, output=R}]}
  MaterialExpressionReroute_2 {class=MaterialExpressionReroute, inputs [None {from=MaterialExpressionNamedRerouteDeclaration_0}]}
  MaterialExpressionReroute_3 {class=MaterialExpressionReroute, inputs [None {from=MaterialExpressionNamedRerouteDeclaration_0}]}
  MaterialExpressionReroute_4 {class=MaterialExpressionReroute, inputs [None {from=MaterialExpressionMultiply_11}]}
  MaterialExpressionScalarParameter_0 {class=MaterialExpressionScalarParameter, properties {defaultValue=1, group=Tex, parameterName=Opacity}}
  MaterialExpressionScalarParameter_1 {class=MaterialExpressionScalarParameter, properties {defaultValue=0.7, group=Tex, parameterName=TexLenMin}}
  MaterialExpressionScalarParameter_10 {class=MaterialExpressionScalarParameter, properties {defaultValue=0.6, group=Displ, parameterName=DisplPanSpeed}}
  MaterialExpressionScalarParameter_11 {class=MaterialExpressionScalarParameter, properties {defaultValue=0.2, group=Displ, parameterName=DisplPanSpeedVar}}
  MaterialExpressionScalarParameter_12 {class=MaterialExpressionScalarParameter, properties {group=Tex, parameterName=Desaturate}}
  MaterialExpressionScalarParameter_13 {class=MaterialExpressionScalarParameter, properties {defaultValue=1, group=Tex, parameterName=ReverseDirectionSwitch}}
  MaterialExpressionScalarParameter_14 {class=MaterialExpressionScalarParameter, properties {defaultValue=3, group=Displ, parameterName=DisplAmpStart}}
  MaterialExpressionScalarParameter_15 {class=MaterialExpressionScalarParameter, properties {defaultValue=5, group=Displ, parameterName=DisplAmpMid}}
  MaterialExpressionScalarParameter_16 {class=MaterialExpressionScalarParameter, properties {defaultValue=1, group=Displ, parameterName=DisplParalAmpMult}}
  MaterialExpressionScalarParameter_17 {class=MaterialExpressionScalarParameter, properties {defaultValue=0.5, group=Displ, parameterName=DisplParalFreqMult}}
  MaterialExpressionScalarParameter_18 {class=MaterialExpressionScalarParameter, properties {defaultValue=1.5, group=Width, parameterName=WidthFreq}}
  MaterialExpressionScalarParameter_19 {class=MaterialExpressionScalarParameter, properties {defaultValue=0.5, group=Width, parameterName=WidthFreqVar}}
  MaterialExpressionScalarParameter_2 {class=MaterialExpressionScalarParameter, properties {defaultValue=1, group=Tex, parameterName=TexLenMax}}
  MaterialExpressionScalarParameter_20 {class=MaterialExpressionScalarParameter, properties {defaultValue=2.5, group=Width, parameterName=WidthAmp}}
  MaterialExpressionScalarParameter_21 {class=MaterialExpressionScalarParameter, properties {defaultValue=0.33, group=Width, parameterName=WidthAmpVar}}
  MaterialExpressionScalarParameter_22 {
   class=MaterialExpressionScalarParameter
   properties {defaultValue=1.2, desc="values over 1 will avoid \r\nline thinning to 0", group=Width, parameterName=WidthThickenOffset}
  }
  MaterialExpressionScalarParameter_23 {class=MaterialExpressionScalarParameter, properties {defaultValue=1.5, group=Width, parameterName=WidthPanSpeed}}
  MaterialExpressionScalarParameter_24 {class=MaterialExpressionScalarParameter, properties {defaultValue=0.6, group=Displ, parameterName=DisplParalPanSpeedMult}}
  MaterialExpressionScalarParameter_25 {class=MaterialExpressionScalarParameter, properties {group=Displ, parameterName=DispParallPanSpeedVar}}
  MaterialExpressionScalarParameter_26 {class=MaterialExpressionScalarParameter, properties {group=Tex, parameterName=TexColorFromAlpha}}
  MaterialExpressionScalarParameter_27 {class=MaterialExpressionScalarParameter, properties {defaultValue=1, group=Width, parameterName=AgeThicknessStart}}
  MaterialExpressionScalarParameter_28 {class=MaterialExpressionScalarParameter, properties {defaultValue=1, group=Width, parameterName=AgeThicknessEnd}}
  MaterialExpressionScalarParameter_3 {class=MaterialExpressionScalarParameter, properties {defaultValue=1, group=Displ, parameterName=DisplAmpEnd}}
  MaterialExpressionScalarParameter_4 {class=MaterialExpressionScalarParameter, properties {defaultValue=0.1, group=Erode, parameterName=ErodeTexFreq}}
  MaterialExpressionScalarParameter_5 {class=MaterialExpressionScalarParameter, properties {defaultValue=0.2, group=Erode, parameterName=ErodeInEndCoord}}
  MaterialExpressionScalarParameter_6 {class=MaterialExpressionScalarParameter, properties {defaultValue=0.8, group=Erode, parameterName=ErodeOutStartCoord}}
  MaterialExpressionScalarParameter_7 {class=MaterialExpressionScalarParameter, properties {defaultValue=0.1, group=Tex, parameterName=PerPartHueVar}}
  MaterialExpressionScalarParameter_8 {class=MaterialExpressionScalarParameter, properties {defaultValue=2.5, group=Displ, parameterName=DisplaceFreq}}
  MaterialExpressionScalarParameter_9 {class=MaterialExpressionScalarParameter, properties {defaultValue=0.3, group=Displ, parameterName=DisplaceFreqVar}}
  MaterialExpressionSine_0 {class=MaterialExpressionSine, inputs [None {from=MaterialExpressionPanner_3}]}
  MaterialExpressionSine_1 {class=MaterialExpressionSine, inputs [None {from=MaterialExpressionConvert_2}]}
  MaterialExpressionSine_5 {class=MaterialExpressionSine, inputs [None {from=MaterialExpressionConvert_7}]}
  MaterialExpressionSmoothStep_0 {
   class=MaterialExpressionSmoothStep
   inputs [
    Min {from="", output=""}
    Max {from=MaterialExpressionScalarParameter_5, output=""}
    Value {from=MaterialExpressionReroute_1, output=""}
   ]
   properties {}
  }
  MaterialExpressionSmoothStep_1 {
   class=MaterialExpressionSmoothStep
   inputs [
    Min {from=MaterialExpressionScalarParameter_6, output=""}
    Max {from="", output=""}
    Value {from=MaterialExpressionReroute_1, output=""}
   ]
   properties {}
  }
  MaterialExpressionStaticBool_0 {class=MaterialExpressionStaticBool}
  MaterialExpressionStaticBool_2 {class=MaterialExpressionStaticBool, properties {value=true}}
  MaterialExpressionSubtract_0 {
   class=MaterialExpressionSubtract
   inputs [
    A {from=""}
    B {from=MaterialExpressionSmoothStep_0}
   ]
  }
  MaterialExpressionSubtract_2 {
   class=MaterialExpressionSubtract
   inputs [
    A {from=MaterialExpressionParticlePositionWS_0}
    B {from=MaterialExpressionDynamicParameter_1, output=RGB}
   ]
  }
  MaterialExpressionSubtract_3 {
   class=MaterialExpressionSubtract
   inputs [
    A {from=""}
    B {from=MaterialExpressionTextureCoordinate_2}
   ]
  }
  MaterialExpressionSubtract_4 {
   class=MaterialExpressionSubtract
   inputs [
    A {from=""}
    B {from=MaterialExpressionConstant_11}
   ]
  }
  MaterialExpressionSubtract_5 {
   class=MaterialExpressionSubtract
   inputs [
    A {from=MaterialExpressionParticlePositionWS_1}
    B {from=MaterialExpressionDynamicParameter_4, output=RGB}
   ]
  }
  MaterialExpressionSubtract_9 {
   class=MaterialExpressionSubtract
   inputs [
    A {from=""}
    B {from=MaterialExpressionConvert_0}
   ]
  }
  MaterialExpressionTextureCoordinate_0 {class=MaterialExpressionTextureCoordinate}
  MaterialExpressionTextureCoordinate_2 {class=MaterialExpressionTextureCoordinate}
  MaterialExpressionTextureSample_1 {
   class=MaterialExpressionTextureSample
   inputs [
    UVs {from=MaterialExpressionAdd_8}
    Tex {from=""}
    Apply View MipBias {from=""}
   ]
   properties {samplerType=SAMPLERTYPE_LinearColor, texture=/Game/Athena/Apollo/Water/Textures/02_Rounded/03_WaterFoam_Veroni_base_H}
  }
  MaterialExpressionTextureSample_2 {
   class=MaterialExpressionTextureSample
   inputs [
    UVs {from=MaterialExpressionReroute_2}
    Tex {from=""}
    Apply View MipBias {from=""}
   ]
   properties {
    samplerSource=SSM_Clamp_WorldGroupSettings
    texture=/BRCosmetics/Athena/Effects/Particles/Skydiving/NS_Contrail_SuperSwirl/Textures/T_Contrail_SuperSwirl_Cloud
   }
  }
  MaterialExpressionTime_1 {class=MaterialExpressionTime}
  MaterialExpressionTime_3 {class=MaterialExpressionTime}
  MaterialExpressionTime_6 {class=MaterialExpressionTime}
 }
 properties {
  bAutomaticallySetUsageInEditor=false
  bUsedWithNiagaraMeshParticles=true
  blendMode=BLEND_Translucent
  translucencyLightingMode=TLM_VolumetricPerVertexDirectional
  usedShadingModels=Default Lit
 }
}
