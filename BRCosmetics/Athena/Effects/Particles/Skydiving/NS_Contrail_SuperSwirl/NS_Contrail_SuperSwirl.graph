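# Niagara skydiving contrail cosmetic that trails swirl and wisp meshes and sprites from the character's skeleton.
# FN_Contrail_Setup supplies direction and velocity to five emitters, plus a dive value that scales DiveSwirls spawn probability.
# SuperSwirl, hand and foot wisps re-sample skeletal sockets each update, then drift along velocity with age, while ambient wisps and dive swirls partially follow mesh movement.
# Dynamic material parameters receive particle position, origin and normalized age.
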NiagaraSystem /BRCosmetics/Athena/Effects/Particles/Skydiving/NS_Contrail_SuperSwirl/NS_Contrail_SuperSwirl
dependencies {
 usedDataInterfaces [&NiagaraDataInterfaceSkeletalMesh]
 usedDynamicInputs [&Lerp_Float]
 usedModules [
  &self:SystemSpawnScript.SystemScriptSource.SystemScriptGraph.NiagaraNodeAssignment_0.SetVariables_#1
  &self:SystemSpawnScript.SystemScriptSource.SystemScriptGraph.NiagaraNodeAssignment_4.SetVariables_#2
  &SystemState
  /Game/Effects/Niagara/Modules/Contrails/FN_Contrail_Setup
  &FN_Integrator
 ]
}
emitters {
 EM_AmbientWisps {
  emitterUpdateScript [
   &EmitterState()
   &SpawnRate(
    Spawn Probability=&RemapRange(Input Value To Remap=$System.FN_Contrail_Setup.Dive Value, Input Min=0, Input Max=1, Output Min=1, Output Max=1)
    SpawnRate=12
    Use Spawn Probability=true
   )
  ]
  particleSpawnScript [
   &InitializeParticle(
    Color=<rgba 1 0 0.0224 1>
    Lifetime=0.85
    Mesh Scale Max=<1.5 1.5 1.5>
    Mesh Scale Mode=::NewEnumerator2
    Sprite Size Mode=::NewEnumerator3
    UsePositionOffset=false
    Write Material Random=true
   ) {module=InitializeParticle001}
   &ShapeLocation(
    Cylinder Height=400
    Cylinder Height Midpoint=1
    Cylinder Radius=300
    MaskSpawnGroup=false
    Offset=<200 0 0>
    Offset Mode=::NewEnumerator0
    Override Seed=false
    Rotation Coordinate Space=::World
    Rotation Mode=::NewEnumerator1
    Shape Origin=&VectorFromFloat(Value=0)
    Shape Primitive=::NewEnumerator1
    Surface Only=true
    Surface Only Band Thickness=0.75
    Use Endcaps In Surface Only Mode=false
    UseOwnerScale=true
   )
   &SkeletalMeshLocation(
    Bone / Socket Index=2
    MaskSpawnGroup=false
    Mesh Sampling Type=::NewEnumerator5
    Orientation Sampling=::NewEnumerator1
    Sampled Position Offset:NiagaraPosition=$Particles.ShapeLocation.ShapePosition
    Skeletal Mesh=$System.Skeletal Mesh
    Socket Sampling Mode=::NewEnumerator5
   ) {module=SkeletalMeshLocation001}
   /SetVariables(Particles.OrigPosition:NiagaraPosition=$Particles.Position) {isSetParameters=true, module=SetVariables_#3}
   &InitialMeshOrientation(
    Mesh Orientation Mode=::NewEnumerator7
    Rotation=&RandomRangeVector(Minimum=<0 0 0>, Maximum=<1 0 0>)
    Use Rotation Vector=true
   )
  ]
  particleUpdateScript [
   &ParticleState()
   &SkeletalMeshLocation(
    Bone / Socket Index=2
    MaskSpawnGroup=false
    Mesh Sampling Type=::NewEnumerator5
    Orientation Sampling=::NewEnumerator1
    Position Sampling=::NewEnumerator1
    Sampled Position Offset:NiagaraPosition=$Particles.Initial.ShapeLocation.ShapePosition
    Skeletal Mesh=$System.Skeletal Mesh
    Socket Sampling Mode=::NewEnumerator5
    Source Movement Scale:Vector3f=&Multiply_VectorByFloat(
     Vector=<1 1 1>
     Float=&Multiply_Float(A=&RandomRangeFloat(Minimum=0.8, Maximum=0.9), B=1)
    )
   )
   &UpdateMeshOrientation(
    Facing Direction=$System.FN_Contrail_Setup.Direction Vector
    Facing Interpolation=false
    Orientation Method=::NewEnumerator1
    Override Quat=false
    TransformFacing=false
    TransformUp=false
   )
   &DynamicMaterialParameters(
    Index 0 Param 4=$Particles.NormalizedAge
    Index 0 Vector 1:Vector3f=&TransformPosition(Position=$Particles.Position, Source Space=::Local, Destination Space=::World)
    Index 1 Vector 1:Vector3f=&TransformPosition(Position=$Particles.OrigPosition, Source Space=::Local, Destination Space=::World)
    Module.Index0Float3Write=true
    Module.Index0VectorWrite=true
    Module.Index1Float3Write=false
    Module.Index1VectorWrite=true
    Module.Index2Float0Write=true
    Module.Index2Float1Write=true
    Module.Index2Float2Write=true
    Module.Index2Float3Write=true
    Write Parameter Index 0=::NewEnumerator3
    Write Parameter Index 1=::NewEnumerator3
    Write Parameter Index 2=::NewEnumerator1
   )
  ]
  rendererClasses [&NiagaraMeshRendererProperties, &NiagaraSpriteRendererProperties]
  renderers [
   {rendererClass=&NiagaraMeshRendererProperties, rendererIndex=0}
   {rendererClass=&NiagaraSpriteRendererProperties, rendererIndex=1}
   {rendererClass=&NiagaraMeshRendererProperties, rendererIndex=2}
  ]
 }
 EM_DiveSwirls {
  emitterUpdateScript [
   &EmitterState()
   &SpawnRate(
    Spawn Probability=&RemapRange(Input Value To Remap=$System.FN_Contrail_Setup.Dive Value, Input Min=0, Input Max=1, Output Min=0, Output Max=1)
    SpawnRate=30
    Use Spawn Probability=true
   )
  ]
  particleSpawnScript [
   &InitializeParticle(
    Color=<rgba 1 0 0.0224 1>
    Lifetime=1.5
    Mesh Scale Max=<0.8 1.75 1.75>
    Mesh Scale Min=<0.5 1.25 1.25>
    Mesh Scale Mode=::NewEnumerator2
    Sprite Size Mode=::NewEnumerator3
    UsePositionOffset=false
    Write Material Random=true
   ) {module=InitializeParticle001}
   /SetVariables(Particles.OrigPosition:NiagaraPosition=$Particles.Position) {isSetParameters=true, module=SetVariables_#3}
   &InitialMeshOrientation(
    Mesh Orientation Mode=::NewEnumerator7
    Rotation=&RandomRangeVector(Minimum=<0 0 0>, Maximum=<1 0 0>)
    Use Rotation Vector=true
   )
  ]
  particleUpdateScript [
   &ParticleState()
   &SkeletalMeshLocation(
    Bone / Socket Index=2
    MaskSpawnGroup=false
    Mesh Sampling Type=::NewEnumerator5
    Orientation Sampling=::NewEnumerator1
    Position Sampling=::NewEnumerator1
    Skeletal Mesh=$System.Skeletal Mesh
    Socket Sampling Mode=::NewEnumerator5
    Source Movement Scale:Vector3f=&Multiply_VectorByFloat(Vector=<1 1 1>, Float=0.985)
   )
   /SetVariables(Particles.OrigPosition:NiagaraPosition=$Particles.Position) {isSetParameters=true, module=SetVariables_#4}
   /SetVariables(Particles.NewAge:NiagaraFloat=&Multiply_Float(A=$Particles.Age, B=1)) {isSetParameters=true, module=SetVariables_#5}
   &UpdateMeshOrientation(Rotation Rate=&RandomRangeFloat(Minimum=0, Maximum=-0.1)) {module=UpdateMeshOrientation001}
   &UpdateMeshOrientation(
    Facing Direction=&Multiply_Vector(A=$System.FN_Contrail_Setup.Direction Velocity, B=<1 1 1>)
    Facing Interpolation=true
    Orientation Method=::NewEnumerator1
    Override Quat=false
    Stabilization Mode:ENiagara_OrientVectorCount=::NewEnumerator0
    TransformFacing=false
   )
   &DynamicMaterialParameters(
    Index 0 Param 4=$Particles.NormalizedAge
    Index 0 Vector 1:Vector3f=&TransformPosition(Position=$Particles.Position, Source Space=::Local, Destination Space=::World)
    Index 1 Vector 1:Vector3f=&TransformPosition(Position=$Particles.OrigPosition, Source Space=::Local, Destination Space=::World)
    Module.Index0Float3Write=true
    Module.Index0VectorWrite=true
    Module.Index1Float3Write=false
    Module.Index1VectorWrite=true
    Module.Index2Float0Write=true
    Module.Index2Float1Write=true
    Module.Index2Float2Write=true
    Module.Index2Float3Write=true
    Write Parameter Index 0=::NewEnumerator3
    Write Parameter Index 1=::NewEnumerator3
    Write Parameter Index 2=::NewEnumerator1
   )
  ]
  rendererClasses [&NiagaraMeshRendererProperties, &NiagaraSpriteRendererProperties]
  renderers [
   {rendererClass=&NiagaraMeshRendererProperties, rendererIndex=0}
   {rendererClass=&NiagaraSpriteRendererProperties, rendererIndex=1}
   {rendererClass=&NiagaraMeshRendererProperties, rendererIndex=2}
  ]
 }
 EM_FootWisps {
  emitterUpdateScript [
   &EmitterState()
   &SpawnRate(SpawnRate=20, Use Spawn Probability=true)
  ]
  particleSpawnScript [
   &InitializeParticle(
    Color=<rgba 1 0 0.0224 1>
    Lifetime=0.4
    Mesh Scale Max=<0.05 0.05 0.05>
    Mesh Scale Min=<0.02 0.02 0.02>
    Mesh Scale Mode=::NewEnumerator2
    Position Mode=::NewEnumerator0
    Sprite Size Mode=::NewEnumerator3
    Write Material Random=true
   ) {module=InitializeParticle001}
   &PartitionParticles(Partition Mode:ENiagara_PartitionMode_v2=::NewEnumerator5)
   &ShapeLocation(
    MaskSpawnGroup=false
    Override Seed=false
    Shape Origin=&VectorFromFloat(Value=0)
    Sphere Radius=4
    UseOwnerScale=true
   )
   &SkeletalMeshLocation(
    Bone / Socket Index=$Particles.PartitionParticles.Partition
    MaskSpawnGroup=false
    Mesh Sampling Type=::NewEnumerator5
    Orientation Sampling=::NewEnumerator1
    Sampled Position Offset:NiagaraPosition=$Particles.ShapeLocation.ShapePosition
    Skeletal Mesh=$System.Skeletal Mesh Feet
    Socket Sampling Mode=::NewEnumerator5
   ) {module=SkeletalMeshLocation001}
   /SetVariables(Particles.OrigPosition:NiagaraPosition=$Particles.Position) {isSetParameters=true, module=SetVariables_#3}
   &InitialMeshOrientation(Mesh Orientation Mode=::NewEnumerator7, Rotation=<0 0 0.5>, Use Rotation Vector=true) {module=InitialMeshOrientation001}
   &InitialMeshOrientation(
    Mesh Orientation Mode=::NewEnumerator7
    Rotation=&RandomRangeVector(Minimum=<0 0 0>, Maximum=<1 0 0>)
    Use Rotation Vector=true
   )
  ]
  particleUpdateScript [
   &ParticleState()
   &SkeletalMeshLocation(
    Bone / Socket Index=$Particles.PartitionParticles.Partition
    MaskSpawnGroup=false
    Mesh Sampling Type=::NewEnumerator5
    Orientation Sampling=::NewEnumerator1
    Sampled Position Offset:NiagaraPosition=$Particles.Initial.ShapeLocation.ShapePosition
    Skeletal Mesh=$System.Skeletal Mesh Feet
    Socket Sampling Mode=::NewEnumerator5
   )
   /SetVariables(Particles.OrigPosition:NiagaraPosition=$Particles.Position) {isSetParameters=true, module=SetVariables_#4}
   /SetVariables(Particles.NewAge:NiagaraFloat=&Multiply_Float(A=$Particles.Age, B=0.75)) {isSetParameters=true, module=SetVariables_#5}
   /SetVariables(Particles.Velocity=&Multiply_VectorByFloat(Vector=&Lerp_Vector(A=<0 0 0>, B=$System.FN_Contrail_Setup.Direction Velocity, Alpha=$Particles.NewAge), Float=1)) {isSetParameters=true, module=SetVariables_#6}
   /SetVariables(Particles.Position=&AddVectorToPosition(
     Position=$Particles.Position
     Vector=&Multiply_VectorByFloat(
      Vector=$Particles.Velocity
      Float=&Multiply_Float(
       A=&Power(a=$Particles.NormalizedAge, b=1.5)
       B=&RandomRangeFloat(Minimum=0.1, Maximum=0.2)
      )
     )
    )) {isSetParameters=true, module=SetVariables_#7}
   /SetVariables(Particles.Velocity=$System.FN_Contrail_Setup.Direction Vector) {isSetParameters=true, module=SetVariables_#8}
   &SolveForcesAndVelocity(Clamp Velocity=false, Limit Acceleration=false)
   &UpdateMeshOrientation(Rotation Rate=-0.8) {module=UpdateMeshOrientation001}
   &UpdateMeshOrientation(
    Facing Direction=$System.FN_Contrail_Setup.Direction Vector
    Facing Interpolation=false
    Orientation Method=::NewEnumerator1
    Override Quat=false
    Side Direction:Vector3f=<1 0 0>
    Stabilization Mode:ENiagara_OrientVectorCount=::NewEnumerator2
    TransformFacing=false
    TransformSide=false
   )
   &DynamicMaterialParameters(
    Index 0 Param 4=$Particles.NormalizedAge
    Index 0 Vector 1:NiagaraPosition=$Particles.Position
    Index 1 Vector 1:NiagaraPosition=$Particles.OrigPosition
    Module.Index0Float3Write=true
    Module.Index0VectorWrite=true
    Module.Index1Float3Write=false
    Module.Index1VectorWrite=true
    Module.Index2Float0Write=true
    Module.Index2Float1Write=true
    Module.Index2Float2Write=true
    Module.Index2Float3Write=true
    Write Parameter Index 0=::NewEnumerator3
    Write Parameter Index 1=::NewEnumerator3
    Write Parameter Index 2=::NewEnumerator1
   )
  ]
  rendererClasses [&NiagaraMeshRendererProperties, &NiagaraSpriteRendererProperties]
  renderers [
   {rendererClass=&NiagaraMeshRendererProperties, rendererIndex=0}
   {rendererClass=&NiagaraSpriteRendererProperties, rendererIndex=1}
   {rendererClass=&NiagaraMeshRendererProperties, rendererIndex=2}
  ]
 }
 EM_HandWisp {
  emitterUpdateScript [
   &EmitterState()
   &SpawnRate(SpawnRate=35, Use Spawn Probability=true)
  ]
  particleSpawnScript [
   &InitializeParticle(
    Color=<rgba 1 0 0.0224 1>
    Mesh Scale Max=<0.2 0.2 0.2>
    Mesh Scale Min=<0.1 0.1 0.1>
    Mesh Scale Mode=::NewEnumerator2
    Position Mode=::NewEnumerator0
    Sprite Size Mode=::NewEnumerator3
    Write Material Random=true
   ) {module=InitializeParticle001}
   &PartitionParticles(Partition Mode:ENiagara_PartitionMode_v2=::NewEnumerator5)
   &ShapeLocation(
    MaskSpawnGroup=false
    Override Seed=false
    Shape Origin=&VectorFromFloat(Value=0)
    Sphere Radius=4
    UseOwnerScale=true
   )
   &SkeletalMeshLocation(
    Bone / Socket Index=$Particles.PartitionParticles.Partition
    MaskSpawnGroup=false
    Mesh Sampling Type=::NewEnumerator5
    Orientation Sampling=::NewEnumerator1
    Sampled Position Offset:NiagaraPosition=$Particles.ShapeLocation.ShapePosition
    Skeletal Mesh=$System.Skeletal Mesh
    Socket Sampling Mode=::NewEnumerator5
   ) {module=SkeletalMeshLocation001}
   /SetVariables(Particles.OrigPosition:NiagaraPosition=$Particles.Position) {isSetParameters=true, module=SetVariables_#3}
   &InitialMeshOrientation(Mesh Orientation Mode=::NewEnumerator7, Rotation=<0 0 0.5>, Use Rotation Vector=true) {module=InitialMeshOrientation001}
   &InitialMeshOrientation(
    Mesh Orientation Mode=::NewEnumerator7
    Rotation=&RandomRangeVector(Minimum=<0 0 0>, Maximum=<1 0 0>)
    Use Rotation Vector=true
   )
  ]
  particleUpdateScript [
   &ParticleState()
   &SkeletalMeshLocation(
    Bone / Socket Index=$Particles.PartitionParticles.Partition
    MaskSpawnGroup=false
    Mesh Sampling Type=::NewEnumerator5
    Orientation Sampling=::NewEnumerator1
    Sampled Position Offset:NiagaraPosition=$Particles.Initial.ShapeLocation.ShapePosition
    Skeletal Mesh=$System.Skeletal Mesh
    Socket Sampling Mode=::NewEnumerator5
   )
   /SetVariables(Particles.OrigPosition:NiagaraPosition=$Particles.Position) {isSetParameters=true, module=SetVariables_#4}
   /SetVariables(Particles.NewAge:NiagaraFloat=&Multiply_Float(A=$Particles.Age, B=0.75)) {isSetParameters=true, module=SetVariables_#5}
   /SetVariables(Particles.Velocity=&Multiply_VectorByFloat(Vector=&Lerp_Vector(A=<0 0 0>, B=$System.FN_Contrail_Setup.Direction Velocity, Alpha=$Particles.NewAge), Float=1)) {isSetParameters=true, module=SetVariables_#6}
   /SetVariables(Particles.Position=&AddVectorToPosition(
     Position=$Particles.Position
     Vector=&Multiply_VectorByFloat(
      Vector=$Particles.Velocity
      Float=&Multiply_Float(
       A=&Power(a=$Particles.NormalizedAge, b=1.5)
       B=&RandomRangeFloat(Minimum=0.03, Maximum=0.1)
      )
     )
    )) {isSetParameters=true, module=SetVariables_#7}
   /SetVariables(Particles.Velocity=$System.FN_Contrail_Setup.Direction Vector) {isSetParameters=true, module=SetVariables_#8}
   &SolveForcesAndVelocity(Clamp Velocity=false, Limit Acceleration=false)
   &UpdateMeshOrientation(Rotation Rate=-0.8) {module=UpdateMeshOrientation001}
   &UpdateMeshOrientation(
    Facing Direction=$System.FN_Contrail_Setup.Direction Vector
    Facing Interpolation=false
    Orientation Method=::NewEnumerator1
    Override Quat=false
    Side Direction:Vector3f=<1 0 0>
    Stabilization Mode:ENiagara_OrientVectorCount=::NewEnumerator2
    TransformFacing=false
    TransformSide=false
   )
   &DynamicMaterialParameters(
    Index 0 Param 4=$Particles.NormalizedAge
    Index 0 Vector 1:NiagaraPosition=$Particles.Position
    Index 1 Vector 1:NiagaraPosition=$Particles.OrigPosition
    Module.Index0Float3Write=true
    Module.Index0VectorWrite=true
    Module.Index1Float3Write=false
    Module.Index1VectorWrite=true
    Module.Index2Float0Write=true
    Module.Index2Float1Write=true
    Module.Index2Float2Write=true
    Module.Index2Float3Write=true
    Write Parameter Index 0=::NewEnumerator3
    Write Parameter Index 1=::NewEnumerator3
    Write Parameter Index 2=::NewEnumerator1
   )
  ]
  rendererClasses [&NiagaraMeshRendererProperties, &NiagaraSpriteRendererProperties]
  renderers [
   {rendererClass=&NiagaraMeshRendererProperties, rendererIndex=0}
   {rendererClass=&NiagaraSpriteRendererProperties, rendererIndex=1}
   {rendererClass=&NiagaraMeshRendererProperties, rendererIndex=2}
  ]
 }
 EM_SuperSwirl {
  emitterUpdateScript [
   &EmitterState()
   &SpawnRate(SpawnRate=45, Use Spawn Probability=true)
   /SetVariables(Emitter.GlobalScale=1) {isSetParameters=true}
  ]
  particleSpawnScript [
   &InitializeParticle(
    Color=<rgba 1 0 0.0224 1>
    Lifetime=1.65
    Mesh Scale Mode=::NewEnumerator1
    Mesh Uniform Scale Max:NiagaraFloat=&Multiply_Float(A=1, B=$Emitter.GlobalScale)
    Mesh Uniform Scale Min:NiagaraFloat=&Multiply_Float(A=1, B=$Emitter.GlobalScale)
    Sprite Size Mode=::NewEnumerator3
    UsePositionOffset=false
    Write Material Random=true
   ) {module=InitializeParticle001}
   &PartitionParticles(Partition Mode:ENiagara_PartitionMode_v2=::NewEnumerator5)
   &ShapeLocation(
    MaskSpawnGroup=false
    Override Seed=false
    Shape Origin=&VectorFromFloat(Value=0)
    Sphere Radius=4
    UseOwnerScale=true
   )
   &SkeletalMeshLocation(
    Bone / Socket Index=$Particles.PartitionParticles.Partition
    MaskSpawnGroup=false
    Mesh Sampling Type=::NewEnumerator5
    Orientation Sampling=::NewEnumerator1
    Sampled Position Offset:NiagaraPosition=$Particles.ShapeLocation.ShapePosition
    Skeletal Mesh=$System.Skeletal Mesh
    Socket Sampling Mode=::NewEnumerator5
   ) {module=SkeletalMeshLocation001}
   /SetVariables(Particles.OrigPosition:NiagaraPosition=$Particles.Position) {isSetParameters=true, module=SetVariables_#3}
  ]
  particleUpdateScript [
   &ParticleState()
   &SkeletalMeshLocation(
    Bone / Socket Index=$Particles.PartitionParticles.Partition
    MaskSpawnGroup=false
    Mesh Sampling Type=::NewEnumerator5
    Orientation Sampling=::NewEnumerator1
    Sampled Position Offset:NiagaraPosition=$Particles.Initial.ShapeLocation.ShapePosition
    Skeletal Mesh=$System.Skeletal Mesh
    Socket Sampling Mode=::NewEnumerator5
   )
   /SetVariables(Particles.OrigPosition:NiagaraPosition=$Particles.Position) {isSetParameters=true, module=SetVariables_#4}
   /SetVariables(Particles.NewAge:NiagaraFloat=&Multiply_Float(A=$Particles.Age, B=0.75)) {isSetParameters=true, module=SetVariables_#5}
   /SetVariables(Particles.Velocity=&Multiply_VectorByFloat(
     Vector=&Lerp_Vector(A=<0 0 0>, B=$System.FN_Contrail_Setup.Direction Velocity, Alpha=$Particles.NewAge)
     Float=$Emitter.GlobalScale
    )) {isSetParameters=true, module=SetVariables_#6}
   /SetVariables(Particles.Position=&AddVectorToPosition(
     Position=$Particles.Position
     Vector=&Multiply_VectorByFloat(
      Vector=$Particles.Velocity
      Float=&Multiply_Float(
       A=&Power(a=$Particles.NormalizedAge, b=0.8)
       B=&RandomRangeFloat(Minimum=0.75, Maximum=1)
      )
     )
    )) {isSetParameters=true, module=SetVariables_#7}
   /SetVariables(Particles.Velocity=$System.FN_Contrail_Setup.Direction Vector) {isSetParameters=true, module=SetVariables_#8}
   &UpdateMeshOrientation(
    Facing Direction=$System.FN_Contrail_Setup.Direction Vector
    Facing Interpolation=false
    Orientation Method=::NewEnumerator1
    Override Quat=false
    TransformFacing=false
    TransformUp=false
   )
   &DynamicMaterialParameters(
    Index 0 Param 4=$Particles.NormalizedAge
    Index 0 Vector 1:NiagaraPosition=$Particles.Position
    Index 1 Vector 1:NiagaraPosition=$Particles.OrigPosition
    Module.Index0Float3Write=true
    Module.Index0VectorWrite=true
    Module.Index1Float3Write=false
    Module.Index1VectorWrite=true
    Module.Index2Float0Write=true
    Module.Index2Float1Write=true
    Module.Index2Float2Write=true
    Module.Index2Float3Write=true
    Write Parameter Index 0=::NewEnumerator3
    Write Parameter Index 1=::NewEnumerator3
    Write Parameter Index 2=::NewEnumerator1
   )
  ]
  rendererClasses [&NiagaraMeshRendererProperties, &NiagaraSpriteRendererProperties]
  renderers [
   {rendererClass=&NiagaraMeshRendererProperties, rendererIndex=0}
   {rendererClass=&NiagaraSpriteRendererProperties, rendererIndex=1}
   {rendererClass=&NiagaraMeshRendererProperties, rendererIndex=2}
  ]
 }
}
summary {
 systemName=NS_Contrail_SuperSwirl
 emitters [
  {emitterName=EM_SuperSwirl, rendererClasses [&NiagaraMeshRendererProperties, &NiagaraSpriteRendererProperties]}
  {emitterName=EM_HandWisp, rendererClasses [&NiagaraMeshRendererProperties, &NiagaraSpriteRendererProperties]}
  {emitterName=EM_FootWisps, rendererClasses [&NiagaraMeshRendererProperties, &NiagaraSpriteRendererProperties]}
  {emitterName=EM_DiveSwirls, rendererClasses [&NiagaraMeshRendererProperties, &NiagaraSpriteRendererProperties]}
  {
   emitterName=EM_AmbientWisps
   rendererClasses [&NiagaraMeshRendererProperties, &NiagaraSpriteRendererProperties]
  }
 ]
 userVariables [
  User.bIsFrontEnd {type=NiagaraBool}
  User.bIsFrontEndPreview {type=NiagaraBool}
  User.ParaGlideLeanAlpha {type=NiagaraFloat, defaultValue=0, description=""}
  User.RibbonIDOffset {type=NiagaraInt32}
 ]
}
